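# Editor utility widget that saves the open assembly level as a PCG data asset and previews it in biome preview levels.
# Saving creates or updates a "<Level>_PCG" asset through the PCG level-to-asset exporter.
# Preview loads the chosen level instance, spawns a transient BP_ActorTagger_Preview, and feeds the asset and UI settings into its PCG graph.
# Refresh re-exports and regenerates; removal flushes PCG cache, deletes preview actors and unloads the level.
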EditorUtilityWidgetBlueprint /PCG/Utilities/Assemblies/ActorTagger/Widgets/Widget_AssemblyPreviewSave
graphs [
 ChangeStatePreviewButtonText {nodes [
   ChangeStatePreviewButtonText() {is=title K2Node_FunctionEntry}
   > n1 = Is Valid(InputObject=Get PreviewLevelInstance() -> (PreviewLevelInstance:Level Streaming Dynamic Object Reference))
   Set ShowPreviewButtonText(execute=@n1.Is Not Valid, ShowPreviewButtonText:String=Show Assembly Preview) -> (Output_Get:String)
   Set ShowPreviewButtonText(execute=@n1.Is Valid, ShowPreviewButtonText=Remove Assembly Preview)
  ]}
 ConstructPCGData {nodes [
   n1 = Inputs!() -> (execute, Outer:Object Reference)
   > n2 = Construct Object from Class(Class=/Script/PCG.PCGPointData, self=@n1) -> (ReturnValue:PCGPoint Data Object Reference)
   > n3 = Set PCGPointDataRoot(PCGPointDataRoot:PCGPoint Data Object Reference=@n2) -> (Output_Get:PCGPoint Data Object Reference)
   > n4 = MutableMetadata!(execute, self:PCGData Object Reference=@n3) -> (then, ReturnValue:PCGMetadata Object Reference)
   > Set PCGMetaDataRoot(PCGMetaDataRoot:PCGMetadata Object Reference=@n4) -> (Output_Get:PCGMetadata Object Reference)
   > n5 = Construct Object from Class(Class=&PCGPointData, self=@n1)
   > n6 = Set PCGPointDataPoints(PCGPointDataPoints:PCGPoint Data Object Reference=@n5) -> (Output_Get:PCGPoint Data Object Reference)
   > n7 = MutableMetadata(self=@n6)
   > Set PCGMetaDataPoints(PCGMetaDataPoints:PCGMetadata Object Reference=@n7) -> (Output_Get:PCGMetadata Object Reference)
   > Outputs!(then:Exec)
  ]}
 CreatePCGDataAsset {nodes [
   n1 = CreatePCGDataAsset!() -> (then, AssetName:String, PackagePath:String, Level:Soft World Reference Structure) {is=title K2Node_FunctionEntry}
   > n2 = GetAssetTools()
   > n3 = CreateAsset(self=@n2, AssetName=@n1, PackagePath=@n1.PackagePath, AssetClass=/Script/PCG.PCGDataAsset)
   > n4 = Set OutputAsset(OutputAsset:Object Reference=@n3) -> (Output_Get:Object Reference)
   > Construct PCGData!(execute:Exec, Outer:Object Reference=@n4) -> (then:Exec) {is=title K2Node_MacroInstance}
   > n5 = Cast To PCGDataAsset!(execute, Object=@n3) -> (then, CastFailed, AsPCGData Asset:PCGData Asset Object Reference) {is=title K2Node_DynamicCast}
   > n6 = Set OutputPCGAsset(OutputPCGAsset:PCGData Asset Object Reference=@n5) -> (Output_Get:PCGData Asset Object Reference)
   > Set Data!(
    execute
    Data:PCGData Collection Structure=Get PCGDataCollectionOutput() -> (PCGDataCollectionOutput:PCGData Collection Structure)
    self:PCGData Asset Object Reference=@n6
   ) -> (then, Output_Get:PCGData Collection Structure)
   > Set Name!(execute, Name:String=Get Name() -> (Name:String), self:PCGData Asset Object Reference=@n5) -> (then, Output_Get:String)
   > n7 = Construct Object from Class(Class=/Script/PCGEditor.PCGLevelToAsset) -> (ReturnValue:PCGLevel To Asset Object Reference)
   n8 = Get Level() -> (Level:Soft World Reference Structure)
   > SetWorld!(
    execute
    self:PCGLevel To Asset Object Reference=@n7
    World:World Object Reference=Resolve Soft Reference!(Input:World Soft Object Reference=Break Soft World Reference!(SoftWorldReference:"Soft World Reference Structure (by ref)"=@n8) -> (WorldAsset:World Soft Object Reference) {is=title K2Node_BreakStruct}) -> (Output:World Object Reference)
   ) -> (then)
   > n9 = CreateAsset!(
    execute
    Exporter:PCGAsset Exporter Object Reference=@n7
    Parameters:PCGAsset Exporter Parameters Structure=Make PCGAsset Exporter Parameters!(
     bOpenSaveDialog:bool=false
     AssetName:String=Get AssetName() -> (AssetName:String)
     AssetPath:String=Get PackagePath() -> (PackagePath:String)
     bSaveOnExportEnded:bool=True
    ) -> (PCGAssetExporterParameters:PCGAsset Exporter Parameters Structure) {is=title K2Node_MakeStruct}
   ) -> (then, ReturnValue:Package Object Reference)
   > n10 = Set OutputAsset(OutputAsset=@n9)
   > Return Node!(execute, Return Value:Object Reference=@n10, Success:bool=true)
   Return Node(execute=@n5.CastFailed, Success=false)
  ]}
 EventGraph {nodes [
   "On Clicked (BtnAssemblySaveToPCGDataAsset)"() {is=title K2Node_ComponentBoundEvent}
   > SaveCurrentLevel!(execute) -> (then, ReturnValue:bool)
   > SaveLevelToPCGDataAsset!(execute, self) -> (then, Success:bool)
   "On Clicked (BtnAssemblyPreview)"() {is=title K2Node_ComponentBoundEvent}
   > n1 = Is Valid(InputObject=Get PreviewLevelInstance())
   n2 = SaveCurrentLevel(execute=@n1.Is Not Valid)
   > n3 = Branch(Condition=@n2)
   > Set with Broadcast PreviewLoaded!(execute, PreviewLoaded:bool=false) -> (then, Output_Get:bool)
   > n4 = GetEditorWorld(self=Get UnrealEditorSubsystem())
   > Set EditorWorld(EditorWorld:Object Reference=@n4) -> (Output_Get:Object Reference)
   > SaveLevelToPCGDataAsset()
   > n5 = SpawnActorFromObject!(
    execute
    ObjectToUse:Object Reference=/PCG/Utilities/Assemblies/ActorTagger/BiomePreviews/BP_ActorTagger_Preview
    Location:Vector="0, 0, 0"
    Rotation:Rotator
    bTransient:bool=true
   ) -> (then, ReturnValue:Actor Object Reference)
   > n6 = Cast To BP_ActorTagger_Preview!(execute, Object=@n5) -> (then, CastFailed, AsBP Actor Tagger Preview:BP Actor Tagger Preview Object Reference) {is=title K2Node_DynamicCast}
   > Set BPActorTaggerPreview(BPActorTaggerPreview:BP Actor Tagger Preview Object Reference=@n6) -> (Output_Get:BP Actor Tagger Preview Object Reference)
   > PrintString(
    InString=Spawned BP_ActorTagger_Preview object
    bPrintToScreen=Get DebugToScreen() -> (DebugToScreen:bool)
    bPrintToLog=Get DebugToLog() -> (DebugToLog:bool)
   )
   n7 = GetSelectedIndex!(self:"Combo Box (String) Object Reference"=Get ComboboxPreviewLevel() -> (ComboboxPreviewLevel:Editor Utility Combo Box String Object Reference)) -> (ReturnValue:int=0)
   n8 = Get PreviewLevels() -> (PreviewLevels:Array of Soft Object References)
   > n9 = "Load Level Instance (by Name)"!(
    execute
    WorldContextObject:Object Reference
    LevelName:String="To String (SoftObjectReference)"!(SoftObjectReference:"Soft Object Reference (by ref)"="Get (a copy)"(Array:Array of Soft Object References=@n8, Dimension 1=@n7) -> (Output:Soft Object Reference)) -> (ReturnValue:String)
    Location:Vector="0, 0, 0"
    Rotation:Rotator="0, 0, 0"
    OptionalLevelNameOverride:String
    OptionalLevelStreamingClass:Level Streaming Dynamic Class Reference
    bLoadAsTempPackage:bool=true
   ) -> (then, bOutSuccess:bool, ReturnValue:Level Streaming Dynamic Object Reference)
   > n10 = Set with Broadcast PreviewLoaded(PreviewLoaded=@n9)
   > n11 = Branch(Condition=@n10)
   > Set with Broadcast PreviewLevelInstance!(execute, PreviewLevelInstance:Level Streaming Dynamic Object Reference=@n9.ReturnValue) -> (then, Output_Get:Level Streaming Dynamic Object Reference)
   n12 = Get EditorWorld() -> (EditorWorld:Object Reference)
   > Set PackagePath(PackagePath:String=Split!(InPath:String=Get Object Path String(Object=@n12)) -> (PathPart:String, FilenamePart:String, ExtensionPart:String)) -> (Output_Get:String)
   > Set AssetName(AssetName:String=Append(A=GetDisplayName(Object=@n12), B=_PCG)) -> (Output_Get:String)
   n13 = Get AssetName()
   > n14 = Set PreviewAssemblyAssetPath(PreviewAssemblyAssetPath:String=Append!(A=Get PackagePath(), B=/, C:String=@n13, D:String=., E:String=@n13) -> (ReturnValue)) -> (Output_Get:String)
   > n15 = LoadAsset!(
    execute
    self:Editor Asset Subsystem Object Reference=Get EditorAssetSubsystem() -> (ReturnValue:Editor Asset Subsystem Object Reference) {is=title K2Node_GetEditorSubsystem}
    AssetPath=@n14
   ) -> (then, ReturnValue)
   > Is Valid(InputObject=@n15)
   > n16 = Cast To PCGDataAsset(Object=@n15)
   n17 = Get BPActorTaggerPreview() -> (BPActorTaggerPreview:BP Actor Tagger Preview Object Reference)
   > n18 = Branch(Condition=Get CanOverridePCGGraph!(self:BP Actor Tagger Preview Object Reference=@n17) -> (CanOverridePCGGraph:bool))
   n19 = Get PCG!(self:BP Actor Tagger Preview Object Reference=@n17) -> (PCG:PCGComponent Object Reference)
   > n20 = SetGraph!(
    execute
    self:PCGComponent Object Reference=@n19
    InGraph:PCGGraph Interface Object Reference=Get PCGPreviewGraph() -> (PCGPreviewGraph:PCGGraph Object Reference)
   ) -> (then)
   n21 = Get GraphInstance!(self:PCGComponent Object Reference=@n19) -> (GraphInstance:PCGGraph Instance Object Reference)
   SetSoftObjectParameter!(
    execute=[@n20.then, @n18.else]
    GraphInterface:PCGGraph Interface Object Reference=@n21
    Name:Name=Settings
    Value:"Soft Object Reference (by ref)"=Make Soft Reference!(Input:PCGData Asset Object Reference=@n16) -> (Output:PCGData Asset Soft Object Reference)
   ) -> (then)
   > SetInt32Parameter!(
    execute
    GraphInterface:PCGGraph Interface Object Reference=@n21
    Name:Name=ClutterSeed
    Value:int=Truncate(A=Get ClutterSeed() -> (ClutterSeed:double))
   ) -> (then)
   > SetFloatParameter!(
    execute
    GraphInterface:PCGGraph Interface Object Reference=@n21
    Name:Name=ClutterDensity
    Value:float=Get ClutterDensity() -> (ClutterDensity:double)
   ) -> (then)
   > SetInt32Parameter(GraphInterface=@n21, Name=MaxPreviewASMInstances, Value=Truncate(A=Get PreviewASMInstanceLimit() -> (PreviewASMInstanceLimit:double)))
   > SetInt32Parameter(GraphInterface=@n21, Name=ASMMaxActorLimit, Value=Truncate(A=Get ASMActorLimit() -> (ASMActorLimit:double)))
   > SetFloatParameter(GraphInterface=@n21, Name=PreviewGridSpacing, Value=Get PreviewGridSpacing() -> (PreviewGridSpacing:double))
   > SetBoolParameter!(
    execute
    GraphInterface:PCGGraph Interface Object Reference=@n21
    Name:Name=GPUMeshSpawn
    bValue:bool=Get UseGPUMeshPreview() -> (UseGPUMeshPreview:bool)
   ) -> (then)
   > Delay!(execute, WorldContextObject:Object Reference, Duration=0.5) -> (then)
   > RefreshGraph!(execute, self:BP Actor Tagger Preview Object Reference=Get BPActorTaggerPreview(), PreviewRefresh:bool=false) -> (then)
   PrintString(
    execute=@n16.CastFailed
    InString=Failed to Cast Preview Asset to PCGDataAsset
    bPrintToScreen=Get DebugToScreen()
    bPrintToLog=Get DebugToLog()
   )
   PrintString(execute=@n11.else, InString=Preview Level Instance failed to load, bPrintToScreen=Get DebugToScreen(), bPrintToLog=Get DebugToLog())
   PrintString(execute=@n3.else, InString="Failed to save the level for Preview ")
   ExecuteConsoleCommand(execute=@n1.Is Valid, WorldContextObject=Get PreviewLevelInstance(), Command=pcg.FlushCache)
   > n22 = GetAllActorsOfClass!(execute, WorldContextObject:Object Reference=Get EditorWorld(), ActorClass:Actor Class Reference=/Script/PCG.PCGWorldActor) -> (then, OutActors:Array of PCGWorld Actor Object References)
   n23 = Get EditorActorSubsystem()
   > n24 = ConvertActors!(
    execute
    self:Editor Actor Subsystem Object Reference=@n23
    Actors:Array of Actor Object References=@n22
    ActorClass:Actor Class Reference=&Actor
    StaticMeshPackagePath:String
   ) -> (then, ReturnValue:Array of Actor Object References)
   > DestroyActors!(execute, self:Editor Actor Subsystem Object Reference=@n23, ActorsToDestroy:Array of Actor Object References=@n24) -> (then, ReturnValue:bool)
   > DestroyActor!(
    execute
    self:Editor Actor Subsystem Object Reference=Get EditorActorSubsystem()
    ActorToDestroy:Actor Object Reference=Get BPActorTaggerPreview()
   ) -> (then, ReturnValue:bool)
   n25 = Get PreviewLevelInstance()
   > SetIsRequestingUnloadAndRemoval!(execute, self:Level Streaming Object Reference=@n25, bInIsRequestingUnloadAndRemoval:bool=true) -> (then)
   > IsLoaded!(execute:Exec, self:Object Reference=@n25) -> (then:Exec, ReturnValue:bool) {is=title K2Node_Message}
   > Set BPActorTaggerPreview()
   > CollectGarbage!(execute) -> (then)
   > n26 = IsLoaded(self=@n25)
   > n27 = Branch(Condition=@n26)
   > Delay(Duration=0.4)
   > n28 = IsLoaded(self=@n25)
   > n29 = Branch(Condition=@n28)
   > PrintString(InString=Failed to unload the Preview level instance)
   Set with Broadcast PreviewLevelInstance(execute=@n29.else)
   > Set with Broadcast PreviewLoaded(PreviewLoaded=false)
   Set with Broadcast PreviewLevelInstance(execute=@n27.else)
   > Set with Broadcast PreviewLoaded(PreviewLoaded=false)
   "On Clicked (BtnReFreshAssemblyPreview)"() {is=title K2Node_ComponentBoundEvent}
   > n30 = Is Valid(InputObject=Get PreviewLevelInstance())
   Set with Broadcast PreviewLoaded(execute=@n30.Is Not Valid, PreviewLoaded=false)
   n31 = SaveCurrentLevel(execute=@n30.Is Valid)
   > n32 = Branch(Condition=@n31)
   > SaveLevelToPCGDataAsset()
   > n33 = LoadAsset(self=Get EditorAssetSubsystem(), AssetPath=Get PreviewAssemblyAssetPath() -> (PreviewAssemblyAssetPath:String))
   > n34 = Is Valid(InputObject=@n33)
   PrintString(execute=@n34.Is Not Valid, InString=Preview PCGData Asset is not valid)
   n35 = Get BPActorTaggerPreview()
   PrintString(
    execute=@n34.Is Valid
    InString=Append!(A="BP_PCG_REF :: ", B=Get Object Path String(Object=@n35)) -> (ReturnValue)
    bPrintToScreen=Get DebugToScreen()
    bPrintToLog=Get DebugToLog()
   )
   > n36 = Cast To PCGDataAsset(Object=@n33)
   > Branch()
   n37 = Get GraphInstance(self=Get PCG(self=Get BPActorTaggerPreview()))
   > n38 = Is Valid(InputObject=@n37)
   PrintString(execute=@n38.Is Not Valid, InString=BP PCG Ref object is not valid)
   SetSoftObjectParameter(execute=@n38.Is Valid, GraphInterface=@n37, Name=Settings, Value=Make Soft Reference(Input=@n36))
   > SetInt32Parameter(GraphInterface=@n37, Name=ClutterSeed, Value=Truncate(A=Get ClutterSeed()))
   > SetFloatParameter(GraphInterface=@n37, Name=ClutterDensity, Value=Get ClutterDensity())
   > SetFloatParameter(GraphInterface=@n37, Name=PreviewGridSpacing, Value=Get PreviewGridSpacing())
   > SetInt32Parameter(GraphInterface=@n37, Name=MaxPreviewASMInstances, Value=Truncate(A=Get PreviewASMInstanceLimit()))
   > SetInt32Parameter(GraphInterface=@n37, Name=ASMMaxActorLimit, Value=Truncate(A=Get ASMActorLimit()))
   > SetBoolParameter(GraphInterface=@n37, Name=GPUMeshSpawn, bValue=Get UseGPUMeshPreview())
   > Delay(Duration=0.45)
   > RefreshGraph(self=Get BPActorTaggerPreview(), PreviewRefresh=true)
   PrintString(execute=@n32.else, InString=Failed to save the level for Preview Refresh)
   n39 = "On Value Changed (EUSpinBoxClutterSeed)"!() -> (OutputDelegate, then, InValue:float) {is=title K2Node_ComponentBoundEvent}
   > Set ClutterSeed(ClutterSeed:double=@n39.InValue) -> (Output_Get:double)
   n40 = "On Value Changed (EUSpinBoxClutterDensity)"!() -> (OutputDelegate, then, InValue:float) {is=title K2Node_ComponentBoundEvent}
   > Set ClutterDensity(ClutterDensity:double=@n40.InValue) -> (Output_Get:double)
   n41 = "On Value Changed (EUSpinBoxAssemblyActorLimit)"!() -> (OutputDelegate, then, InValue:float) {is=title K2Node_ComponentBoundEvent}
   > Set ASMActorLimit(ASMActorLimit:double=@n41.InValue) -> (Output_Get:double)
   n42 = "On Value Changed (EUSpinBoxAssemblyInstanceLimit)"!() -> (OutputDelegate, then, InValue:float) {is=title K2Node_ComponentBoundEvent}
   > Set PreviewASMInstanceLimit(PreviewASMInstanceLimit:double=@n42.InValue) -> (Output_Get:double)
   n43 = "On Value Changed (EUSpinBoxPreviewGridSpacing)"!() -> (OutputDelegate, then, InValue:float) {is=title K2Node_ComponentBoundEvent}
   > Set PreviewGridSpacing(PreviewGridSpacing:double=@n43.InValue) -> (Output_Get:double)
   n44 = IsPreviewLevelInstanceValid!() -> (OutputDelegate, then, Object:Object Reference, Field:Field Notification Id Structure) {is=title K2Node_CustomEvent}
   > ChangeStatePreviewButtonText() {is=title K2Node_CallFunction}
   n45 = Update!() -> (OutputDelegate, then, MapChangeEventFlags:int) {is=title K2Node_CustomEvent}
   > UpdateOnMapChanged()
   Event Construct()
   n46 = Get SinglePropertyViewPCGPreviewGraph() -> (SinglePropertyViewPCGPreviewGraph:Single Property View Object Reference)
   > SetObject(self=@n46, NewObject=Self-Reference())
   > SetPropertyName!(execute, self:Single Property View Object Reference=@n46, NewPropertyName:Name=PCGPreviewGraph) -> (then)
   > PopulateInternalAssets()
   > Add Field Value Changed Delegate!(execute, self:Widget Object Reference, FieldId:Field Notification Id Structure=(FieldName="PreviewLevelInstance"), Delegate:Delegate=@n44) -> (then)
   n47 = Get LevelEditorSubsystem()
   > Unbind Event from On Map Changed(self:Level Editor Subsystem Object Reference=@n47, Delegate=@n45) {is=title K2Node_RemoveDelegate}
   > Bind Event to On Map Changed(self:Level Editor Subsystem Object Reference=@n47, Delegate=@n45) {is=title K2Node_AddDelegate}
   Event PreConstruct()
   > Branch()
   n48 = Get SinglePropertyViewPCGPreviewGraph()
   > SetObject(self=@n48, NewObject=Self-Reference())
   > SetPropertyName(self=@n48, NewPropertyName=PCGPreviewGraph)
   Delay(Duration=0.45)
  ]}
 PopulateDropDownfromFolder {nodes [
   n1 = PopulateDropDownfromFolder!() -> (then, ComboBoxString:"Combo Box (String) Object Reference", FolderPath:String) {is=title K2Node_FunctionEntry}
   > Clear(TargetArray:Array of Asset Data Structures=Get AssetDataPreviewFolder() -> (AssetDataPreviewFolder:Array of Asset Data Structures))
   > Clear(TargetArray:Array of Strings=Get Local_ObjectReferences() -> (Local_ObjectReferences:Array of Strings))
   > n2 = GetAssetsByPath!(
    execute:Exec
    self:Object Reference="To Object (Interface)"!(Interface:"Interface Interface (by ref)"=GetAssetRegistry()) -> (ReturnValue:Object Reference)
    PackagePath:Name=String To Name(InString=@n1.FolderPath)
    bRecursive:bool=false
    bIncludeOnlyOnDiskAssets:bool=true
   ) -> (then:Exec, OutAssetData:Array of Asset Data Structures, ReturnValue:bool) {is=title K2Node_Message}
   > n3 = Sequence!(execute) -> (then_0, then_1, then_2:Exec, then_3:Exec, then_4:Exec, then_5:Exec, then_6:Exec, then_7:Exec)
   > n4 = Set AssetDataPreviewFolder(AssetDataPreviewFolder:Array of Asset Data Structures=@n2) -> (Output_Get:Array of Asset Data Structures)
   > n5 = For Each Loop(Array:Array of Asset Data Structures=@n4) -> (Array Element:Asset Data Structure)
   n6 = "To String (Name)"(InName=Break Asset Data(AssetData=@n5).AssetName)
   n7 = Get Class Display Name(Class=GetClass!(InAssetData:"Asset Data Structure (by ref)"=@n5) -> (ReturnValue:Object Class Reference))
   > n8 = Branch(Condition=OR Boolean(A="Equal Exactly (String)"(A=@n7, B=World), B="Equal Exactly (String)"(A=@n6, B=BP_ActorTagger_Preview_C)))
   Remove Index!(execute=@n8.else, TargetArray:Array of Asset Data Structures=@n4, IndexToRemove:int=@n5.Array Index) -> (then)
   SortByAssetName!(execute=@n3.then_1, Assets:Array of Asset Data Structures=Get AssetDataPreviewFolder(), SortOrder:EAssetRegistrySortOrder Enum=Ascending) -> (then)
   > n9 = For Each Loop(Array=Get AssetDataPreviewFolder())
   n10 = GetClass(InAssetData=@n9)
   > Branch(Condition="Equal, Case Insensitive (String)"(A=Get Class Display Name(Class=@n10), B=World))
   > Add(TargetArray=Get Local_ObjectReferences(), NewItem="To String (Name)"(InName=Break Asset Data(AssetData=@n9).AssetName)) {is=title K2Node_CallArrayFunction}
   n11 = Get Local_ObjectReferences()
   SortStringArray!(execute=@n3.then_2, TargetArray:Array of Strings=@n11, bStableSort:bool=false, SortOrder:EArraySortOrder Enum=Ascending) -> (then)
   > n12 = For Each Loop(Array:Array of Strings=@n11) -> (Array Element:String)
   > n13 = For Each Loop(Array:Array of Asset Data Structures=Get AssetDataPreviewFolder()) -> (Array Element:Asset Data Structure)
   n14 = Break Asset Data(AssetData=@n13)
   > Branch(Condition="Equal Exactly (String)"(A="To String (Name)"(InName=@n14.AssetName), B=@n12))
   > Add!(
    execute
    TargetMap:Map of Strings to Asset Data Structures=Get OptionMapping() -> (OptionMapping:Map of Strings to Asset Data Structures)
    Key:"String (by ref)"=@n12
    Value:"Asset Data Structure (by ref)"=@n13
   ) -> (then) {is=title K2Node_CallFunction}
   n15 = Get OptionMapping()
   n16 = Keys!(execute=@n3.then_3, TargetMap:Map of Strings to Asset Data Structures=@n15) -> (then, Keys:Array of Strings)
   > n17 = Values!(execute, TargetMap:Map of Strings to Asset Data Structures=@n15) -> (then, Values:Array of Asset Data Structures)
   > n18 = For Each Loop(Array:Array of Strings=@n16) -> (Array Element:String)
   > PrintString(InString=Append!(A="To String (Integer)"(InInt=@n18.Array Index), B=::, C:String=@n18) -> (ReturnValue), bPrintToLog=Get DebugToScreen())
   n19 = "Get (a copy)"(Array:Array of Asset Data Structures=@n17, Dimension 1=@n18.Array Index) -> (Output:Asset Data Structure)
   > PrintString(InString="To String (Name)"(InName=Break Asset Data(AssetData=@n19).AssetName), bPrintToScreen=false, bPrintToLog=Get DebugToScreen())
   PrintString(execute=@n3.then_4, InString="\r\n\r\n", bPrintToScreen=false, bPrintToLog=Get DebugToScreen())
   > n20 = For Each Loop(Array:Array of Asset Data Structures=Get AssetDataPreviewFolder()) -> (Array Element:Asset Data Structure)
   n21 = Break Asset Data(AssetData=@n20)
   > PrintString(
    InString=Append(A="To String (Integer)"(InInt=@n20.Array Index), B=":: ", C="To String (Name)"(InName=@n21))
    bPrintToScreen=false
    bPrintToLog=Get DebugToScreen()
   )
   ClearOptions(
    execute=@n3.then_5
    self:"Combo Box (String) Object Reference"=Get ComboBoxString() -> (ComboBoxString:"Combo Box (String) Object Reference")
   )
   > Clear(TargetArray:Array of Soft Object References=Get PreviewLevels())
   n22 = Get OptionMapping()
   > Keys(TargetMap=@n22)
   > n23 = Values(TargetMap=@n22)
   > n24 = For Each Loop(Array=@n23)
   n25 = GetClass(InAssetData=@n24)
   > Branch(Condition="Equal, Case Insensitive (String)"(A=Get Class Display Name(Class=@n25), B=World))
   n26 = Break Asset Data(AssetData=@n24)
   n27 = Get ComboBoxString()
   > AddOption!(execute, self:"Combo Box (String) Object Reference"=@n27, Option:String="To String (Name)"(InName=@n26.AssetName)) -> (then)
   n28 = "To String (Name)"(InName=@n26)
   > Add Unique(
    TargetArray:Array of Soft Object References=Get PreviewLevels()
    NewItem:"Soft Object Reference (by ref)"=To Soft Object Reference(SoftObjectPath=MakeSoftObjectPath!(PathString:String=@n28) -> (ReturnValue:Soft Object Path Structure))
   )
   > RefreshOptions(self:"Combo Box (String) Object Reference"=@n27)
   > SetSelectedIndex!(execute, self:"Combo Box (String) Object Reference"=@n27, Index:int=0) -> (then)
   Return Node!(execute=@n3.then_7)
   n29 = Get ComboBoxString()
   RefreshOptions(execute=@n3.then_6, self=@n29)
   > SetSelectedIndex(self=@n29, Index=0)
  ]}
 PopulateInternalAssets {nodes [
   PopulateInternalAssets() {is=title K2Node_FunctionEntry}
   > n1 = Is Valid(InputObject=GetParent!(self:Widget Object Reference=Self-Reference()) -> (ReturnValue:Panel Widget Object Reference))
   n2 = GetSoftClassPath!(Class:Object Class Reference=Get Class(Object=Self-Reference()) -> (ReturnValue:Widget Assembly Preview Save Class Reference)) -> (ReturnValue:Soft Class Path Structure)
   n3 = Split!(
    SourceString:String=Split(SourceString=BreakSoftClassPath!(InSoftClassPath:Soft Class Path Structure=@n2) -> (PathString:String), InStr=/, SearchDir=FromEnd)
    InStr:String=/
    SearchCase:ESearchCase Enum
    SearchDir:ESearchDir Enum=FromEnd
   ) -> (LeftS:String, RightS:String, ReturnValue:bool)
   n4 = Set ActorTaggerFolderPath(execute=@n1.Is Not Valid, ActorTaggerFolderPath:String=@n3) -> (Output_Get:String)
   > n5 = Set BiomePreviewsFolderPath(BiomePreviewsFolderPath:String=Append(A=@n4, B=/, C=Get PreviewLevelsFolderName() -> (PreviewLevelsFolderName:String))) -> (Output_Get:String)
   Set UserPreviewFolderPath(
    execute=[@n5.then, @n9.then]
    UserPreviewFolderPath:String=Append(A=Get ActorTaggerFolderPath() -> (ActorTaggerFolderPath:String), B=/, C=UserPreview)
   ) -> (Output_Get:String)
   > PopulateDropDownfromFolder!(
    execute
    self
    ComboBoxString:"Combo Box (String) Object Reference"=Get ComboboxPreviewLevel()
    FolderPath:String=Get BiomePreviewsFolderPath() -> (BiomePreviewsFolderPath:String)
   ) -> (then) {is=title K2Node_CallFunction}
   > n6 = LoadAsset!(execute, AssetPath=Append(A=Get ActorTaggerFolderPath(), B=/, C=BiomePreviews/PCG_ActorTagger_PreviewGraph)) -> (then, ReturnValue)
   > n7 = Cast To PCGGraph!(execute, Object=@n6) -> (then, CastFailed, AsPCGGraph:PCGGraph Object Reference) {is=title K2Node_DynamicCast}
   > Set PCGPreviewGraph(PCGPreviewGraph:PCGGraph Object Reference=@n7) -> (Output_Get:PCGGraph Object Reference)
   > Set ReferenceLightingLevelPath(ReferenceLightingLevelPath:String=Append(A=Get ActorTaggerFolderPath(), B=/, C=Reference/Reference_Lighting)) -> (Output_Get:String)
   n8 = Set ActorTaggerFolderPath(execute=@n1.Is Valid, ActorTaggerFolderPath=@n3)
   > n9 = Set BiomePreviewsFolderPath(BiomePreviewsFolderPath=Append(A=@n8, B=/, C=Get PreviewLevelsFolderName()))
  ]}
 SaveLevelToPCGDataAsset {nodes [
   SaveLevelToPCGDataAsset() {is=title K2Node_FunctionEntry}
   > n1 = GetEditorWorld(self=Get UnrealEditorSubsystem())
   > n2 = Set Level(Level:Object Reference=@n1) -> (Output_Get:Object Reference)
   > Set PackagePath(PackagePath=Split!(InPath:String=Get Object Path String(Object=@n2)) -> (PathPart:String, FilenamePart:String, ExtensionPart:String))
   > Set AssetName(AssetName=Append!(A=GetDisplayName(Object=@n2), B=_PCG) -> (ReturnValue))
   n3 = Get AssetName()
   > Set AssetPath(AssetPath:String=Append!(A=Get PackagePath(), B=/, C:String=@n3, D:String=., E:String=@n3) -> (ReturnValue)) -> (Output_Get:String)
   > n4 = DoesAssetExist(AssetPath=Get AssetPath() -> (AssetPath:String))
   > n5 = Branch(Condition=@n4)
   > PrintString(InString="Preview Asset Exists ->Loading Preview Asset")
   > n6 = LoadAsset(AssetPath=Get AssetPath())
   > n7 = Cast To PCGDataAsset(Object=@n6)
   n8 = MakeSoftObjectPath(PathString=Get Object Path String(Object=Get Level() -> (Level:Object Reference)))
   > Set ObjectPath!(execute, ObjectPath:Soft Object Path Structure=@n8, self:PCGData Asset Object Reference=@n7) -> (then, Output_Get:Soft Object Path Structure)
   n9 = Make PCGAsset Exporter Parameters(bOpenSaveDialog=false, AssetName=Get AssetName(), AssetPath=Get PackagePath(), bSaveOnExportEnded=True)
   > UpdateAssets!(
    execute
    PCGAssets:Array of Asset Data Structures=Make Array("[0]":Asset Data Structure=CreateAssetData(InAsset=@n7, bAllowBlueprintClass=false)) -> (Array:Array of Asset Data Structures)
    Parameters:PCGAsset Exporter Parameters Structure=@n9
   ) -> (then)
   > Return Node!(execute, Success:bool=true)
   PrintString(execute=[@n11.CastFailed, @n7.CastFailed], InString=Failed to Cast to PCGDataAsset)
   > Return Node(Success=false)
   PrintString(execute=@n5.else, InString="Preview Asset Not Found -> Creating New Preview Asset")
   > n10 = CreatePCGDataAsset!(
    execute
    self
    AssetName:String=Get AssetName()
    PackagePath:String=Get PackagePath()
    Level:Soft World Reference Structure=Make Soft World Reference!(WorldAsset:World Soft Object Reference=Make Soft Reference(Input:World Object Reference=@n1) -> (Output:World Soft Object Reference)) -> (SoftWorldReference:Soft World Reference Structure) {is=title K2Node_MakeStruct}
   ) -> (then, Return Value:Object Reference, Success:bool) {is=title K2Node_CallFunction}
   > n11 = Cast To PCGDataAsset(Object=@n10)
   > Set ObjectPath(ObjectPath=@n8, self=@n11)
   > UpdateAssets(PCGAssets=Make Array("[0]"=CreateAssetData(InAsset=@n11, bAllowBlueprintClass=false)), Parameters=@n9)
   > Return Node(Success=true)
  ]}
 UpdateOnMapChanged {nodes [UpdateOnMapChanged() {is=title K2Node_FunctionEntry}]}
]
variables [
 ActorTaggerFolderPath {type=string, default=""}
 ASMActorLimit {type=number, default=1500}
 AssetName {type=string, default=""}
 BiomePreviewsFolderPath {type=string, default=""}
 BPActorTaggerPreview {type=BP_ActorTagger_Preview_C, default=null}
 ClutterDensity {type=number, default=0.66}
 ClutterSeed {type=number, default=101}
 DebugToLog {type=boolean}
 DebugToScreen {type=boolean}
 EditorWorld {type=Object, default=null}
 PackagePath {type=string, default=""}
 PCGMetaDataPoints {type=PCGMetadata, default=null}
 PCGMetaDataRoot {type=PCGMetadata, default=null}
 PCGPointDataPoints {type=PCGPointData, default=null}
 PCGPointDataRoot {type=PCGPointData, default=null}
 PCGPreviewGraph {type=PCGGraph, default=null}
 PreviewASMInstanceLimit {type=number, default=512}
 PreviewAssemblyAssetPath {type=string, default=""}
 PreviewGridSpacing {type=number, default=1}
 PreviewLevelInstance {type=LevelStreamingDynamic, default=null}
 PreviewLevels {type="array<Object>", default []}
 PreviewLevelsFolderName {type=string, default=BiomePreviews}
 PreviewLoaded {type=boolean}
 ReferenceLightingLevelPath {type=string, default=""}
 ShowPreviewButtonText {type=string, default=Show Assembly Preview}
 UseGPUMeshPreview {type=boolean}
 UserPreviewFolderPath {type=string, default=""}
]
widgets {
 info {rootWidgetClass=&Border, widgetCount=27}
 widgets [
  Border_244 {widgetClassPath=&Border, widget=Border_244, properties {verticalAlignment=VAlign_Top, brushColor=<rgba 0.0219 0.0219 0.0219 1>}}
  VerticalBox_401 {
   widgetClassPath=&VerticalBox
   parent=Border_244
   widget=VerticalBox_401
   slot {class=BorderSlot, properties {verticalAlignment=VAlign_Top}}
  }
  UniformGridPanel_01 {widgetClassPath=/Script/UMG.UniformGridPanel, parent=VerticalBox_401, widget=UniformGridPanel_01, slot {class=VerticalBoxSlot}}
  TextBlock_22 {
   widgetClassPath=&TextBlock
   parent=UniformGridPanel_01
   widget=TextBlock_22
   properties {
    lineHeightPercentage=0.8
    text=Assembly Preview Level
    font {
     bForceMonospaced=false
     bMaterialIsStencil=false
     fontMaterial=null
     fontObject=&Roboto
     letterSpacing=0
     monospacedWidth=1
     size=8
     skewAmount=0.12
     typefaceFontName=Regular
     outlineSettings {
      bApplyOutlineToDropShadows=false
      bMiteredCorners=false
      bSeparateFillAlpha=false
      outlineMaterial=null
      outlineSize=0
      outlineColor=<rgba 0 0 0 1>
     }
    }
    margin=<ltrb 3 3 3 3>
   }
   slot {class=UniformGridSlot, properties {horizontalAlignment=HAlign_Center, verticalAlignment=VAlign_Center}}
  }
  ComboboxPreviewLevel {
   widgetClassPath=/Script/Blutility.EditorUtilityComboBoxString
   bIsVariable=true
   parent=UniformGridPanel_01
   widget=ComboboxPreviewLevel
   properties {
    enableGamepadNavigationMode=false
    font {
     bForceMonospaced=false
     bMaterialIsStencil=false
     fontMaterial=null
     fontObject=&Roboto
     letterSpacing=0
     monospacedWidth=1
     size=10
     skewAmount=0
     typefaceFontName=Regular
     outlineSettings {
      bApplyOutlineToDropShadows=false
      bMiteredCorners=false
      bSeparateFillAlpha=false
      outlineMaterial=null
      outlineSize=0
      outlineColor=<rgba 0 0 0 1>
     }
    }
    foregroundColor {colorUseRule=UseColor_Foreground, specifiedColor=<rgba 0.9092 0.9092 0.9092 1>}
    itemStyle {
     bUseParentRowBrush=false
     activeBrush {
      bIsDynamicallyLoaded=false
      drawAs=Image
      imageType=NoImage
      mirroring=NoMirror
      resourceName=None
      resourceObject=null
      tiling=NoTile
      imageSize=<0 0>
      margin=<ltrb 0 0 0 0>
      outlineSettings {
       bUseBrushTransparency=false
       roundingType=HalfHeightRadius
       width=0
       color {colorUseRule=UseColor_Specified, specifiedColor=<rgba 0 0 0 0>}
       cornerRadii=<0 0 0 0>
      }
      tintColor {colorUseRule=UseColor_Specified, specifiedColor=<rgba 0 0.162 0.7454 1>}
      uVRegion {bIsValid=false, max=<0 0>, min=<0 0>}
     }
     activeHighlightedBrush {
      bIsDynamicallyLoaded=false
      drawAs=Image
      imageType=NoImage
      mirroring=NoMirror
      resourceName=None
      resourceObject=null
      tiling=NoTile
      imageSize=<0 0>
      margin=<ltrb 0 0 0 0>
      outlineSettings {
       bUseBrushTransparency=false
       roundingType=HalfHeightRadius
       width=0
       color {colorUseRule=UseColor_Specified, specifiedColor=<rgba 0 0 0 0>}
       cornerRadii=<0 0 0 0>
      }
      tintColor {colorUseRule=UseColor_Specified, specifiedColor=<rgba 0.0252 0.0319 0.0423 1>}
      uVRegion {bIsValid=false, max=<0 0>, min=<0 0>}
     }
     activeHoveredBrush {
      bIsDynamicallyLoaded=false
      drawAs=Image
      imageType=NoImage
      mirroring=NoMirror
      resourceName=None
      resourceObject=null
      tiling=NoTile
      imageSize=<0 0>
      margin=<ltrb 0 0 0 0>
      outlineSettings {
       bUseBrushTransparency=false
       roundingType=HalfHeightRadius
       width=0
       color {colorUseRule=UseColor_Specified, specifiedColor=<rgba 0 0 0 0>}
       cornerRadii=<0 0 0 0>
      }
      tintColor {colorUseRule=UseColor_Specified, specifiedColor=<rgba 0 0.162 0.7454 1>}
      uVRegion {bIsValid=false, max=<0 0>, min=<0 0>}
     }
     dropIndicator_Above {
      bIsDynamicallyLoaded=false
      drawAs=Box
      imageType=FullColor
      mirroring=NoMirror
      resourceName=../../../Engine/Content/Slate/Common/DropZoneIndicator_Above.png
      resourceObject=null
      tiling=NoTile
      imageSize=<0 0>
      margin=<ltrb 0.625 0.625 0 0>
      outlineSettings {
       bUseBrushTransparency=false
       roundingType=HalfHeightRadius
       width=0
       color {colorUseRule=UseColor_Specified, specifiedColor=<rgba 0 0 0 0>}
       cornerRadii=<0 0 0 0>
      }
      tintColor {colorUseRule=UseColor_Specified, specifiedColor=<rgba 0.0091 0.3515 0.9647 1>}
      uVRegion {bIsValid=false, max=<0 0>, min=<0 0>}
     }
     dropIndicator_Below {
      bIsDynamicallyLoaded=false
      drawAs=Box
      imageType=FullColor
      mirroring=NoMirror
      resourceName=../../../Engine/Content/Slate/Common/DropZoneIndicator_Below.png
      resourceObject=null
      tiling=NoTile
      imageSize=<0 0>
      margin=<ltrb 0.625 0 0 0.625>
      outlineSettings {
       bUseBrushTransparency=false
       roundingType=HalfHeightRadius
       width=0
       color {colorUseRule=UseColor_Specified, specifiedColor=<rgba 0 0 0 0>}
       cornerRadii=<0 0 0 0>
      }
      tintColor {colorUseRule=UseColor_Specified, specifiedColor=<rgba 0.0091 0.3515 0.9647 1>}
      uVRegion {bIsValid=false, max=<0 0>, min=<0 0>}
     }
     dropIndicator_Onto {
      bIsDynamicallyLoaded=false
      drawAs=Box
      imageType=FullColor
      mirroring=NoMirror
      resourceName=../../../Engine/Content/Slate/Common/DropZoneIndicator_Onto.png
      resourceObject=null
      tiling=NoTile
      imageSize=<0 0>
      margin=<ltrb 0.25 0.25 0.25 0.25>
      outlineSettings {
       bUseBrushTransparency=false
       roundingType=HalfHeightRadius
       width=0
       color {colorUseRule=UseColor_Specified, specifiedColor=<rgba 0 0 0 0>}
       cornerRadii=<0 0 0 0>
      }
      tintColor {colorUseRule=UseColor_Specified, specifiedColor=<rgba 0.0091 0.3515 0.9647 1>}
      uVRegion {bIsValid=false, max=<0 0>, min=<0 0>}
     }
     evenRowBackgroundBrush {
      bIsDynamicallyLoaded=false
      drawAs=Image
      imageType=NoImage
      mirroring=NoMirror
      resourceName=None
      resourceObject=null
      tiling=NoTile
      imageSize=<0 0>
      margin=<ltrb 0 0 0 0>
      outlineSettings {
       bUseBrushTransparency=false
       roundingType=HalfHeightRadius
       width=0
       color {colorUseRule=UseColor_Specified, specifiedColor=<rgba 0 0 0 0>}
       cornerRadii=<0 0 0 0>
      }
      tintColor {colorUseRule=UseColor_Specified, specifiedColor=<rgba 0.0103 0.0103 0.0103 1>}
      uVRegion {bIsValid=false, max=<0 0>, min=<0 0>}
     }
     evenRowBackgroundHoveredBrush {
      bIsDynamicallyLoaded=false
      drawAs=Image
      imageType=NoImage
      mirroring=NoMirror
      resourceName=None
      resourceObject=null
      tiling=NoTile
      imageSize=<0 0>
      margin=<ltrb 0 0 0 0>
      outlineSettings {
       bUseBrushTransparency=false
       roundingType=HalfHeightRadius
       width=0
       color {colorUseRule=UseColor_Specified, specifiedColor=<rgba 0 0 0 0>}
       cornerRadii=<0 0 0 0>
      }
      tintColor {colorUseRule=UseColor_Specified, specifiedColor=<rgba 0.0176 0.0176 0.0176 1>}
      uVRegion {bIsValid=false, max=<0 0>, min=<0 0>}
     }
     inactiveBrush {
      bIsDynamicallyLoaded=false
      drawAs=Image
      imageType=NoImage
      mirroring=NoMirror
      resourceName=None
      resourceObject=null
      tiling=NoTile
      imageSize=<0 0>
      margin=<ltrb 0 0 0 0>
      outlineSettings {
       bUseBrushTransparency=false
       roundingType=HalfHeightRadius
       width=0
       color {colorUseRule=UseColor_Specified, specifiedColor=<rgba 0 0 0 0>}
       cornerRadii=<0 0 0 0>
      }
      tintColor {colorUseRule=UseColor_Specified, specifiedColor=<rgba 0.0513 0.0953 0.159 1>}
      uVRegion {bIsValid=false, max=<0 0>, min=<0 0>}
     }
     inactiveHighlightedBrush {
      bIsDynamicallyLoaded=false
      drawAs=Image
      imageType=NoImage
      mirroring=NoMirror
      resourceName=None
      resourceObject=null
      tiling=NoTile
      imageSize=<0 0>
      margin=<ltrb 0 0 0 0>
      outlineSettings {
       bUseBrushTransparency=false
       roundingType=HalfHeightRadius
       width=0
       color {colorUseRule=UseColor_Specified, specifiedColor=<rgba 0 0 0 0>}
       cornerRadii=<0 0 0 0>
      }
      tintColor {colorUseRule=UseColor_Specified, specifiedColor=<rgba 0.0252 0.0319 0.0423 1>}
      uVRegion {bIsValid=false, max=<0 0>, min=<0 0>}
     }
     inactiveHoveredBrush {
      bIsDynamicallyLoaded=false
      drawAs=Image
      imageType=NoImage
      mirroring=NoMirror
      resourceName=None
      resourceObject=null
      tiling=NoTile
      imageSize=<0 0>
      margin=<ltrb 0 0 0 0>
      outlineSettings {
       bUseBrushTransparency=false
       roundingType=HalfHeightRadius
       width=0
       color {colorUseRule=UseColor_Specified, specifiedColor=<rgba 0 0 0 0>}
       cornerRadii=<0 0 0 0>
      }
      tintColor {colorUseRule=UseColor_Specified, specifiedColor=<rgba 0.0513 0.0953 0.159 1>}
      uVRegion {bIsValid=false, max=<0 0>, min=<0 0>}
     }
     oddRowBackgroundBrush {
      bIsDynamicallyLoaded=false
      drawAs=Image
      imageType=NoImage
      mirroring=NoMirror
      resourceName=None
      resourceObject=null
      tiling=NoTile
      imageSize=<0 0>
      margin=<ltrb 0 0 0 0>
      outlineSettings {
       bUseBrushTransparency=false
       roundingType=HalfHeightRadius
       width=0
       color {colorUseRule=UseColor_Specified, specifiedColor=<rgba 0 0 0 0>}
       cornerRadii=<0 0 0 0>
      }
      tintColor {colorUseRule=UseColor_Specified, specifiedColor=<rgba 0.0103 0.0103 0.0103 1>}
      uVRegion {bIsValid=false, max=<0 0>, min=<0 0>}
     }
     oddRowBackgroundHoveredBrush {
      bIsDynamicallyLoaded=false
      drawAs=Image
      imageType=NoImage
      mirroring=NoMirror
      resourceName=None
      resourceObject=null
      tiling=NoTile
      imageSize=<0 0>
      margin=<ltrb 0 0 0 0>
      outlineSettings {
       bUseBrushTransparency=false
       roundingType=HalfHeightRadius
       width=0
       color {colorUseRule=UseColor_Specified, specifiedColor=<rgba 0 0 0 0>}
       cornerRadii=<0 0 0 0>
      }
      tintColor {colorUseRule=UseColor_Specified, specifiedColor=<rgba 0.0176 0.0176 0.0176 1>}
      uVRegion {bIsValid=false, max=<0 0>, min=<0 0>}
     }
     parentRowBackgroundBrush {
      bIsDynamicallyLoaded=false
      drawAs=Image
      imageType=NoImage
      mirroring=NoMirror
      resourceName=None
      resourceObject=null
      tiling=NoTile
      imageSize=<32 32>
      margin=<ltrb 0 0 0 0>
      outlineSettings {
       bUseBrushTransparency=false
       roundingType=HalfHeightRadius
       width=0
       color {colorUseRule=UseColor_Specified, specifiedColor=<rgba 0 0 0 0>}
       cornerRadii=<0 0 0 0>
      }
      tintColor {colorUseRule=UseColor_Specified, specifiedColor=<rgba 1 1 1 1>}
      uVRegion {bIsValid=false, max=<0 0>, min=<0 0>}
     }
     parentRowBackgroundHoveredBrush {
      bIsDynamicallyLoaded=false
      drawAs=Image
      imageType=NoImage
      mirroring=NoMirror
      resourceName=None
      resourceObject=null
      tiling=NoTile
      imageSize=<32 32>
      margin=<ltrb 0 0 0 0>
      outlineSettings {
       bUseBrushTransparency=false
       roundingType=HalfHeightRadius
       width=0
       color {colorUseRule=UseColor_Specified, specifiedColor=<rgba 0 0 0 0>}
       cornerRadii=<0 0 0 0>
      }
      tintColor {colorUseRule=UseColor_Specified, specifiedColor=<rgba 1 1 1 1>}
      uVRegion {bIsValid=false, max=<0 0>, min=<0 0>}
     }
     selectedTextColor {colorUseRule=UseColor_Specified, specifiedColor=<rgba 1 1 1 1>}
     selectorFocusedBrush {
      bIsDynamicallyLoaded=false
      drawAs=Border
      imageType=FullColor
      mirroring=NoMirror
      resourceName=../../../Engine/Content/Slate/Common/Selector.png
      resourceObject=null
      tiling=Both
      imageSize=<0 0>
      margin=<ltrb 0.25 0.25 0.25 0.25>
      outlineSettings {
       bUseBrushTransparency=false
       roundingType=HalfHeightRadius
       width=0
       color {colorUseRule=UseColor_Specified, specifiedColor=<rgba 0 0 0 0>}
       cornerRadii=<0 0 0 0>
      }
      tintColor {colorUseRule=UseColor_Specified, specifiedColor=<rgba 0.701 0.225 0.003 1>}
      uVRegion {bIsValid=false, max=<0 0>, min=<0 0>}
     }
     textColor {colorUseRule=UseColor_Specified, specifiedColor=<rgba 0.5271 0.5271 0.5271 1>}
    }
    widgetStyle {
     comboButtonStyle {
      downArrowAlign=VAlign_Top
      buttonStyle {
       clickedSlateSound {resourceObject=null}
       disabled {
        bIsDynamicallyLoaded=false
        drawAs=RoundedBox
        imageType=NoImage
        mirroring=NoMirror
        resourceName=None
        resourceObject=null
        tiling=NoTile
        imageSize=<0 0>
        margin=<ltrb 0 0 0 0>
        outlineSettings {
         bUseBrushTransparency=false
         roundingType=FixedRadius
         width=1
         color {colorUseRule=UseColor_ColorTable, specifiedColor=<rgba 0 0 0 0>}
         cornerRadii=<4 4 4 4>
        }
        tintColor {colorUseRule=UseColor_Specified, specifiedColor=<rgba 0.0395 0.0395 0.0395 1>}
        uVRegion {bIsValid=false, max=<0 0>, min=<0 0>}
       }
       disabledForeground {colorUseRule=UseColor_ColorTable, specifiedColor=<rgba 0 0 0 0>}
       hovered {
        bIsDynamicallyLoaded=false
        drawAs=RoundedBox
        imageType=NoImage
        mirroring=NoMirror
        resourceName=None
        resourceObject=null
        tiling=NoTile
        imageSize=<0 0>
        margin=<ltrb 0 0 0 0>
        outlineSettings {
         bUseBrushTransparency=false
         roundingType=FixedRadius
         width=1
         color {colorUseRule=UseColor_ColorTable, specifiedColor=<rgba 0 0 0 0>}
         cornerRadii=<4 4 4 4>
        }
        tintColor {colorUseRule=UseColor_Specified, specifiedColor=<rgba 0.0048 0.0048 0.0048 1>}
        uVRegion {bIsValid=false, max=<0 0>, min=<0 0>}
       }
       hoveredForeground {colorUseRule=UseColor_ColorTable, specifiedColor=<rgba 0 0 0 0>}
       hoveredSlateSound {resourceObject=null}
       normal {
        bIsDynamicallyLoaded=false
        drawAs=RoundedBox
        imageType=NoImage
        mirroring=NoMirror
        resourceName=None
        resourceObject=null
        tiling=NoTile
        imageSize=<0 0>
        margin=<ltrb 0 0 0 0>
        outlineSettings {
         bUseBrushTransparency=false
         roundingType=FixedRadius
         width=1
         color {colorUseRule=UseColor_ColorTable, specifiedColor=<rgba 0 0 0 0>}
         cornerRadii=<4 4 4 4>
        }
        tintColor {colorUseRule=UseColor_Specified, specifiedColor=<rgba 0.0048 0.0048 0.0048 1>}
        uVRegion {bIsValid=false, max=<0 0>, min=<0 0>}
       }
       normalForeground {colorUseRule=UseColor_ColorTable, specifiedColor=<rgba 0 0 0 0>}
       normalPadding=<ltrb 8 1 8 1>
       pressed {
        bIsDynamicallyLoaded=false
        drawAs=RoundedBox
        imageType=NoImage
        mirroring=NoMirror
        resourceName=None
        resourceObject=null
        tiling=NoTile
        imageSize=<0 0>
        margin=<ltrb 0 0 0 0>
        outlineSettings {
         bUseBrushTransparency=false
         roundingType=FixedRadius
         width=1
         color {colorUseRule=UseColor_ColorTable, specifiedColor=<rgba 0 0 0 0>}
         cornerRadii=<4 4 4 4>
        }
        tintColor {colorUseRule=UseColor_Specified, specifiedColor=<rgba 0.0048 0.0048 0.0048 1>}
        uVRegion {bIsValid=false, max=<0 0>, min=<0 0>}
       }
       pressedForeground {colorUseRule=UseColor_ColorTable, specifiedColor=<rgba 0 0 0 0>}
       pressedPadding=<ltrb 8 2 8 0>
       pressedSlateSound {resourceObject=null}
      }
      contentPadding=<ltrb 0 0 0 0>
      downArrowImage {
       bIsDynamicallyLoaded=false
       drawAs=Image
       imageType=Vector
       mirroring=NoMirror
       resourceName=../../../Engine/Content/Slate/Starship/CoreWidgets/ComboBox/wide-chevron-down.svg
       resourceObject=null
       tiling=NoTile
       imageSize=<16 16>
       margin=<ltrb 0 0 0 0>
       outlineSettings {
        bUseBrushTransparency=false
        roundingType=HalfHeightRadius
        width=0
        color {colorUseRule=UseColor_Specified, specifiedColor=<rgba 0 0 0 0>}
        cornerRadii=<0 0 0 0>
       }
       tintColor {colorUseRule=UseColor_Specified, specifiedColor=<rgba 1 1 1 1>}
       uVRegion {bIsValid=false, max=<0 0>, min=<0 0>}
      }
      downArrowPadding=<ltrb 2 2 2 2>
      menuBorderBrush {
       bIsDynamicallyLoaded=false
       drawAs=RoundedBox
       imageType=NoImage
       mirroring=NoMirror
       resourceName=None
       resourceObject=null
       tiling=NoTile
       imageSize=<0 0>
       margin=<ltrb 0 0 0 0>
       outlineSettings {
        bUseBrushTransparency=false
        roundingType=FixedRadius
        width=1
        color {colorUseRule=UseColor_ColorTable, specifiedColor=<rgba 0 0 0 0>}
        cornerRadii=<0 0 0 0>
       }
       tintColor {colorUseRule=UseColor_Specified, specifiedColor=<rgba 0.0395 0.0395 0.0395 1>}
       uVRegion {bIsValid=false, max=<0 0>, min=<0 0>}
      }
      menuBorderPadding=<ltrb 0 0 0 0>
      shadowColorAndOpacity=<rgba 0 0 0 1>
      shadowOffset=<0 0>
     }
     contentPadding=<ltrb 0 0 0 0>
     menuRowPadding=<ltrb 8 3 8 3>
     pressedSlateSound {resourceObject=null}
     selectionChangeSlateSound {resourceObject=null}
    }
   }
   slot {class=UniformGridSlot, properties {horizontalAlignment=HAlign_Center, row=1, verticalAlignment=VAlign_Fill}}
  }
  Spacer_16 {
   widgetClassPath=&Spacer
   parent=VerticalBox_401
   widget=Spacer_16
   properties {size=<1 4>}
   slot {class=VerticalBoxSlot, properties {verticalAlignment=VAlign_Center}}
  }
  GridPanel {
   widgetClassPath=&GridPanel
   parent=VerticalBox_401
   widget=GridPanel
   properties {visibility=Visible, columnFill [1,1,1,0]}
   slot {class=VerticalBoxSlot}
  }
  BtnAssemblyPreview {
   widgetClassPath=&Button
   parent=GridPanel
   widget=BtnAssemblyPreview
   properties {
    toolTipText=Preview current assembly actors with PCG behaviors on a landscape.
    widgetStyle {
     clickedSlateSound {resourceObject=null}
     disabled {
      bIsDynamicallyLoaded=false
      drawAs=RoundedBox
      imageType=NoImage
      mirroring=NoMirror
      resourceName=None
      resourceObject=null
      tiling=NoTile
      imageSize=<0 0>
      margin=<ltrb 0 0 0 0>
      outlineSettings {
       bUseBrushTransparency=false
       roundingType=FixedRadius
       width=1
       color {colorUseRule=UseColor_ColorTable, specifiedColor=<rgba 0 0 0 0>}
       cornerRadii=<4 4 4 4>
      }
      tintColor {colorUseRule=UseColor_Specified, specifiedColor=<rgba 0.0395 0.0395 0.0395 1>}
      uVRegion {bIsValid=false, max=<0 0>, min=<0 0>}
     }
     disabledForeground {colorUseRule=UseColor_ColorTable, specifiedColor=<rgba 0 0 0 0>}
     hovered {
      bIsDynamicallyLoaded=false
      drawAs=RoundedBox
      imageType=NoImage
      mirroring=NoMirror
      resourceName=None
      resourceObject=null
      tiling=NoTile
      imageSize=<32 32>
      margin=<ltrb 0 0 0 0>
      outlineSettings {
       bUseBrushTransparency=false
       roundingType=FixedRadius
       width=1
       color {colorUseRule=UseColor_ColorTable, specifiedColor=<rgba 0 0 0 0>}
       cornerRadii=<4 4 4 4>
      }
      tintColor {colorUseRule=UseColor_Specified, specifiedColor=<rgba 0.0953 0.0953 0.0953 1>}
      uVRegion {bIsValid=false, max=<0 0>, min=<0 0>}
     }
     hoveredForeground {colorUseRule=UseColor_ColorTable, specifiedColor=<rgba 0 0 0 0>}
     hoveredSlateSound {resourceObject=null}
     normal {
      bIsDynamicallyLoaded=false
      drawAs=RoundedBox
      imageType=NoImage
      mirroring=NoMirror
      resourceName=None
      resourceObject=null
      tiling=NoTile
      imageSize=<32 32>
      margin=<ltrb 0 0 0 0>
      outlineSettings {
       bUseBrushTransparency=false
       roundingType=FixedRadius
       width=1
       color {colorUseRule=UseColor_ColorTable, specifiedColor=<rgba 0 0 0 0>}
       cornerRadii=<4 4 4 4>
      }
      tintColor {colorUseRule=UseColor_Specified, specifiedColor=<rgba 0.0395 0.0395 0.0395 1>}
      uVRegion {bIsValid=false, max=<0 0>, min=<0 0>}
     }
     normalForeground {colorUseRule=UseColor_ColorTable, specifiedColor=<rgba 0 0 0 0>}
     normalPadding=<ltrb 12 1.5 12 1.5>
     pressed {
      bIsDynamicallyLoaded=false
      drawAs=RoundedBox
      imageType=NoImage
      mirroring=NoMirror
      resourceName=None
      resourceObject=null
      tiling=NoTile
      imageSize=<32 32>
      margin=<ltrb 0 0 0 0>
      outlineSettings {
       bUseBrushTransparency=false
       roundingType=FixedRadius
       width=1
       color {colorUseRule=UseColor_ColorTable, specifiedColor=<rgba 0 0 0 0>}
       cornerRadii=<4 4 4 4>
      }
      tintColor {colorUseRule=UseColor_Specified, specifiedColor=<rgba 0.0284 0.0284 0.0284 1>}
      uVRegion {bIsValid=false, max=<0 0>, min=<0 0>}
     }
     pressedForeground {colorUseRule=UseColor_ColorTable, specifiedColor=<rgba 0 0 0 0>}
     pressedPadding=<ltrb 12 2.5 12 0.5>
     pressedSlateSound {resourceObject=null}
    }
   }
   slot {class=GridSlot}
  }
  TextBlock_23 {
   widgetClassPath=&TextBlock
   parent=BtnAssemblyPreview
   widget=TextBlock_23
   properties {
    justification=Center
    text=Show Assembly Preview
    font {
     bForceMonospaced=false
     bMaterialIsStencil=false
     fontMaterial=null
     fontObject=&Roboto
     letterSpacing=0
     monospacedWidth=1
     size=12
     skewAmount=0
     typefaceFontName=Regular
     outlineSettings {
      bApplyOutlineToDropShadows=false
      bMiteredCorners=false
      bSeparateFillAlpha=false
      outlineMaterial=null
      outlineSize=0
      outlineColor=<rgba 0 0 0 1>
     }
    }
   }
   slot {class=ButtonSlot}
  }
  BtnReFreshAssemblyPreview {
   widgetClassPath=&Button
   parent=GridPanel
   widget=BtnReFreshAssemblyPreview
   properties {
    bIsEnabled=false
    toolTipText="Refresh assembly preview when in Preview mode.\r\nButton is only enabled in Preview mode."
    widgetStyle {
     clickedSlateSound {resourceObject=null}
     disabled {
      bIsDynamicallyLoaded=false
      drawAs=RoundedBox
      imageType=NoImage
      mirroring=NoMirror
      resourceName=None
      resourceObject=null
      tiling=NoTile
      imageSize=<0 0>
      margin=<ltrb 0 0 0 0>
      outlineSettings {
       bUseBrushTransparency=false
       roundingType=FixedRadius
       width=1
       color {colorUseRule=UseColor_ColorTable, specifiedColor=<rgba 0 0 0 0>}
       cornerRadii=<4 4 4 4>
      }
      tintColor {colorUseRule=UseColor_Specified, specifiedColor=<rgba 0.0395 0.0395 0.0395 1>}
      uVRegion {bIsValid=false, max=<0 0>, min=<0 0>}
     }
     disabledForeground {colorUseRule=UseColor_ColorTable, specifiedColor=<rgba 0 0 0 0>}
     hovered {
      bIsDynamicallyLoaded=false
      drawAs=RoundedBox
      imageType=NoImage
      mirroring=NoMirror
      resourceName=None
      resourceObject=null
      tiling=NoTile
      imageSize=<32 32>
      margin=<ltrb 0 0 0 0>
      outlineSettings {
       bUseBrushTransparency=false
       roundingType=FixedRadius
       width=1
       color {colorUseRule=UseColor_ColorTable, specifiedColor=<rgba 0 0 0 0>}
       cornerRadii=<4 4 4 4>
      }
      tintColor {colorUseRule=UseColor_Specified, specifiedColor=<rgba 0.0953 0.0953 0.0953 1>}
      uVRegion {bIsValid=false, max=<0 0>, min=<0 0>}
     }
     hoveredForeground {colorUseRule=UseColor_ColorTable, specifiedColor=<rgba 0 0 0 0>}
     hoveredSlateSound {resourceObject=null}
     normal {
      bIsDynamicallyLoaded=false
      drawAs=RoundedBox
      imageType=NoImage
      mirroring=NoMirror
      resourceName=None
      resourceObject=null
      tiling=NoTile
      imageSize=<32 32>
      margin=<ltrb 0 0 0 0>
      outlineSettings {
       bUseBrushTransparency=false
       roundingType=FixedRadius
       width=1
       color {colorUseRule=UseColor_ColorTable, specifiedColor=<rgba 0 0 0 0>}
       cornerRadii=<4 4 4 4>
      }
      tintColor {colorUseRule=UseColor_Specified, specifiedColor=<rgba 0.0395 0.0395 0.0395 1>}
      uVRegion {bIsValid=false, max=<0 0>, min=<0 0>}
     }
     normalForeground {colorUseRule=UseColor_ColorTable, specifiedColor=<rgba 0 0 0 0>}
     normalPadding=<ltrb 12 1.5 12 1.5>
     pressed {
      bIsDynamicallyLoaded=false
      drawAs=RoundedBox
      imageType=NoImage
      mirroring=NoMirror
      resourceName=None
      resourceObject=null
      tiling=NoTile
      imageSize=<32 32>
      margin=<ltrb 0 0 0 0>
      outlineSettings {
       bUseBrushTransparency=false
       roundingType=FixedRadius
       width=1
       color {colorUseRule=UseColor_ColorTable, specifiedColor=<rgba 0 0 0 0>}
       cornerRadii=<4 4 4 4>
      }
      tintColor {colorUseRule=UseColor_Specified, specifiedColor=<rgba 0.0284 0.0284 0.0284 1>}
      uVRegion {bIsValid=false, max=<0 0>, min=<0 0>}
     }
     pressedForeground {colorUseRule=UseColor_ColorTable, specifiedColor=<rgba 0 0 0 0>}
     pressedPadding=<ltrb 12 2.5 12 0.5>
     pressedSlateSound {resourceObject=null}
    }
   }
   slot {class=GridSlot, properties {column=1, columnSpan=1}}
  }
  TextBlock_33 {
   widgetClassPath=&TextBlock
   parent=BtnReFreshAssemblyPreview
   widget=TextBlock_33
   properties {
    text=Refresh Preview
    font {
     bForceMonospaced=false
     bMaterialIsStencil=false
     fontMaterial=null
     fontObject=&Roboto
     letterSpacing=0
     monospacedWidth=1
     size=12
     skewAmount=0
     typefaceFontName=Italic
     outlineSettings {
      bApplyOutlineToDropShadows=false
      bMiteredCorners=false
      bSeparateFillAlpha=false
      outlineMaterial=null
      outlineSize=0
      outlineColor=<rgba 0 0 0 1>
     }
    }
   }
   slot {class=ButtonSlot}
  }
  Border_4 {
   widgetClassPath=&Border
   parent=GridPanel
   widget=Border_4
   properties {brushColor=<rgba 0.0135 0.0135 0.0135 1>, padding=<ltrb 0 0 0 0>}
   slot {class=GridSlot, properties {columnSpan=2, row=8, rowSpan=1}}
  }
  SinglePropertyViewPCGPreviewGraph {
   widgetClassPath=&SinglePropertyView
   parent=Border_4
   widget=SinglePropertyViewPCGPreviewGraph
   properties {bAutoLoadAsset=false, bIsVolatile=true, propertyName=ActorClass}
   slot {class=BorderSlot, properties {padding=<ltrb 0 0 0 0>}}
  }
  Border_8 {
   widgetClassPath=&Border
   parent=GridPanel
   widget=Border_8
   properties {brushColor=<rgba 0.0135 0.0135 0.0135 1>, padding=<ltrb 0 0 0 0>}
   slot {class=GridSlot, properties {columnSpan=2, row=11, rowSpan=1}}
  }
  EUSpinBoxClutterDensity {
   widgetClassPath=/Script/Blutility.EditorUtilitySpinBox
   bIsVariable=true
   parent=GridPanel
   widget=EUSpinBoxClutterDensity
   properties {
    bOverride_MaxSliderValue=true
    bOverride_MaxValue=true
    bOverride_MinSliderValue=true
    bOverride_MinValue=true
    justification=Center
    maxFractionalDigits=2
    maxSliderValue=1.5
    maxValue=1.5
    minFractionalDigits=2
    foregroundColor {colorUseRule=UseColor_Specified, specifiedColor=<rgba 1 1 1 1>}
   }
   slot {class=GridSlot, properties {column=1, row=5}}
  }
  EUSpinBoxPreviewGridSpacing {
   widgetClassPath=&EditorUtilitySpinBox
   bIsVariable=true
   parent=GridPanel
   widget=EUSpinBoxPreviewGridSpacing
   properties {
    bAlwaysUsesDeltaSnap=true
    bOverride_MaxSliderValue=true
    bOverride_MaxValue=true
    bOverride_MinSliderValue=true
    bOverride_MinValue=true
    justification=Center
    maxFractionalDigits=2
    maxSliderValue=3
    maxValue=3
    minSliderValue=0.1
    minValue=0.1
    foregroundColor {colorUseRule=UseColor_Specified, specifiedColor=<rgba 1 1 1 1>}
   }
   slot {class=GridSlot, properties {row=7}}
  }
  EUSpinBoxClutterSeed {
   widgetClassPath=&EditorUtilitySpinBox
   bIsVariable=true
   parent=GridPanel
   widget=EUSpinBoxClutterSeed
   properties {
    bOverride_MaxSliderValue=true
    bOverride_MaxValue=true
    bOverride_MinSliderValue=true
    bOverride_MinValue=true
    justification=Center
    maxFractionalDigits=0
    maxSliderValue=1024
    maxValue=64000
    minFractionalDigits=0
    minSliderValue=-1024
    minValue=-64000
    foregroundColor {colorUseRule=UseColor_Specified, specifiedColor=<rgba 1 1 1 1>}
   }
   slot {class=GridSlot, properties {row=5}}
  }
  EUSpinBoxAssemblyActorLimit {
   widgetClassPath=&EditorUtilitySpinBox
   bIsVariable=true
   parent=GridPanel
   widget=EUSpinBoxAssemblyActorLimit
   properties {
    bOverride_MaxSliderValue=true
    bOverride_MaxValue=true
    bOverride_MinSliderValue=true
    bOverride_MinValue=true
    justification=Center
    maxFractionalDigits=0
    maxSliderValue=2048
    maxValue=2048
    minFractionalDigits=0
    minSliderValue=1
    minValue=1
    value=64
    foregroundColor {colorUseRule=UseColor_Specified, specifiedColor=<rgba 1 1 1 1>}
   }
   slot {class=GridSlot, properties {row=2}}
  }
  EUSpinBoxAssemblyInstanceLimit {
   widgetClassPath=&EditorUtilitySpinBox
   bIsVariable=true
   parent=GridPanel
   widget=EUSpinBoxAssemblyInstanceLimit
   properties {
    bOverride_MaxSliderValue=true
    bOverride_MaxValue=true
    bOverride_MinSliderValue=true
    bOverride_MinValue=true
    justification=Center
    maxFractionalDigits=0
    maxSliderValue=1024
    maxValue=1024
    minFractionalDigits=0
    value=32
    foregroundColor {colorUseRule=UseColor_Specified, specifiedColor=<rgba 1 1 1 1>}
   }
   slot {class=GridSlot, properties {column=1, row=2}}
  }
  TextBlock_36 {
   widgetClassPath=&TextBlock
   parent=GridPanel
   widget=TextBlock_36
   properties {
    lineHeightPercentage=0.8
    text=Clutter Seed
    font {
     bForceMonospaced=false
     bMaterialIsStencil=false
     fontMaterial=null
     fontObject=&Roboto
     letterSpacing=0
     monospacedWidth=1
     size=9
     skewAmount=0
     typefaceFontName=Regular
     outlineSettings {
      bApplyOutlineToDropShadows=false
      bMiteredCorners=false
      bSeparateFillAlpha=false
      outlineMaterial=null
      outlineSize=0
      outlineColor=<rgba 0 0 0 1>
     }
    }
    margin=<ltrb 3 3 3 3>
   }
   slot {class=GridSlot, properties {horizontalAlignment=HAlign_Center, row=3, verticalAlignment=VAlign_Center}}
  }
  TextBlock_57 {
   widgetClassPath=&TextBlock
   parent=GridPanel
   widget=TextBlock_57
   properties {
    lineHeightPercentage=0.8
    text=Preview Grid Spacing
    font {
     bForceMonospaced=false
     bMaterialIsStencil=false
     fontMaterial=null
     fontObject=&Roboto
     letterSpacing=0
     monospacedWidth=1
     size=9
     skewAmount=0
     typefaceFontName=Regular
     outlineSettings {
      bApplyOutlineToDropShadows=false
      bMiteredCorners=false
      bSeparateFillAlpha=false
      outlineMaterial=null
      outlineSize=0
      outlineColor=<rgba 0 0 0 1>
     }
    }
    margin=<ltrb 3 3 3 3>
   }
   slot {class=GridSlot, properties {horizontalAlignment=HAlign_Center, row=6, verticalAlignment=VAlign_Center}}
  }
  TextBlock_53 {
   widgetClassPath=&TextBlock
   parent=GridPanel
   widget=TextBlock_53
   properties {
    lineHeightPercentage=0.8
    text=Assembly Actor Limit
    font {
     bForceMonospaced=false
     bMaterialIsStencil=false
     fontMaterial=null
     fontObject=&Roboto
     letterSpacing=0
     monospacedWidth=1
     size=9
     skewAmount=0
     typefaceFontName=Regular
     outlineSettings {
      bApplyOutlineToDropShadows=false
      bMiteredCorners=false
      bSeparateFillAlpha=false
      outlineMaterial=null
      outlineSize=0
      outlineColor=<rgba 0 0 0 1>
     }
    }
    margin=<ltrb 3 3 3 3>
   }
   slot {class=GridSlot, properties {horizontalAlignment=HAlign_Center, row=1, verticalAlignment=VAlign_Center}}
  }
  TextBlock_54 {
   widgetClassPath=&TextBlock
   parent=GridPanel
   widget=TextBlock_54
   properties {
    lineHeightPercentage=0.8
    text=Preview Assembly Count Limit
    font {
     bForceMonospaced=false
     bMaterialIsStencil=false
     fontMaterial=null
     fontObject=&Roboto
     letterSpacing=0
     monospacedWidth=1
     size=9
     skewAmount=0
     typefaceFontName=Regular
     outlineSettings {
      bApplyOutlineToDropShadows=false
      bMiteredCorners=false
      bSeparateFillAlpha=false
      outlineMaterial=null
      outlineSize=0
      outlineColor=<rgba 0 0 0 1>
     }
    }
    margin=<ltrb 3 3 3 3>
   }
   slot {class=GridSlot, properties {column=1, horizontalAlignment=HAlign_Center, row=1, verticalAlignment=VAlign_Center}}
  }
  TextBlock_37 {
   widgetClassPath=&TextBlock
   parent=GridPanel
   widget=TextBlock_37
   properties {
    lineHeightPercentage=0.8
    text=Clutter Density
    font {
     bForceMonospaced=false
     bMaterialIsStencil=false
     fontMaterial=null
     fontObject=&Roboto
     letterSpacing=0
     monospacedWidth=1
     size=9
     skewAmount=0
     typefaceFontName=Regular
     outlineSettings {
      bApplyOutlineToDropShadows=false
      bMiteredCorners=false
      bSeparateFillAlpha=false
      outlineMaterial=null
      outlineSize=0
      outlineColor=<rgba 0 0 0 1>
     }
    }
    margin=<ltrb 3 3 3 3>
   }
   slot {class=GridSlot, properties {column=1, horizontalAlignment=HAlign_Center, row=3, verticalAlignment=VAlign_Center}}
  }
  Spacer_10 {
   widgetClassPath=&Spacer
   parent=VerticalBox_401
   widget=Spacer_10
   properties {size=<1 8>}
   slot {class=VerticalBoxSlot, properties {horizontalAlignment=HAlign_Left}}
  }
  BtnAssemblySaveToPCGDataAsset {
   widgetClassPath=&Button
   parent=VerticalBox_401
   widget=BtnAssemblySaveToPCGDataAsset
   properties {
    toolTipText="Export assembly actors to a  PCG Data Asset, equivalent to using \"Create PCG Assets from Level(s)\" Asset Action Script."
    widgetStyle {
     clickedSlateSound {resourceObject=null}
     disabled {
      bIsDynamicallyLoaded=false
      drawAs=RoundedBox
      imageType=NoImage
      mirroring=NoMirror
      resourceName=None
      resourceObject=null
      tiling=NoTile
      imageSize=<0 0>
      margin=<ltrb 0 0 0 0>
      outlineSettings {
       bUseBrushTransparency=false
       roundingType=FixedRadius
       width=1
       color {colorUseRule=UseColor_ColorTable, specifiedColor=<rgba 0 0 0 0>}
       cornerRadii=<4 4 4 4>
      }
      tintColor {colorUseRule=UseColor_Specified, specifiedColor=<rgba 0.0395 0.0395 0.0395 1>}
      uVRegion {bIsValid=false, max=<0 0>, min=<0 0>}
     }
     disabledForeground {colorUseRule=UseColor_ColorTable, specifiedColor=<rgba 0 0 0 0>}
     hovered {
      bIsDynamicallyLoaded=false
      drawAs=RoundedBox
      imageType=NoImage
      mirroring=NoMirror
      resourceName=None
      resourceObject=null
      tiling=NoTile
      imageSize=<32 32>
      margin=<ltrb 0 0 0 0>
      outlineSettings {
       bUseBrushTransparency=false
       roundingType=FixedRadius
       width=1
       color {colorUseRule=UseColor_ColorTable, specifiedColor=<rgba 0 0 0 0>}
       cornerRadii=<4 4 4 4>
      }
      tintColor {colorUseRule=UseColor_Specified, specifiedColor=<rgba 0.0953 0.0953 0.0953 1>}
      uVRegion {bIsValid=false, max=<0 0>, min=<0 0>}
     }
     hoveredForeground {colorUseRule=UseColor_ColorTable, specifiedColor=<rgba 0 0 0 0>}
     hoveredSlateSound {resourceObject=null}
     normal {
      bIsDynamicallyLoaded=false
      drawAs=RoundedBox
      imageType=NoImage
      mirroring=NoMirror
      resourceName=None
      resourceObject=null
      tiling=NoTile
      imageSize=<32 32>
      margin=<ltrb 0 0 0 0>
      outlineSettings {
       bUseBrushTransparency=false
       roundingType=FixedRadius
       width=1
       color {colorUseRule=UseColor_ColorTable, specifiedColor=<rgba 0 0 0 0>}
       cornerRadii=<4 4 4 4>
      }
      tintColor {colorUseRule=UseColor_Specified, specifiedColor=<rgba 0.0395 0.0395 0.0395 1>}
      uVRegion {bIsValid=false, max=<0 0>, min=<0 0>}
     }
     normalForeground {colorUseRule=UseColor_ColorTable, specifiedColor=<rgba 0 0 0 0>}
     normalPadding=<ltrb 12 1.5 12 1.5>
     pressed {
      bIsDynamicallyLoaded=false
      drawAs=RoundedBox
      imageType=NoImage
      mirroring=NoMirror
      resourceName=None
      resourceObject=null
      tiling=NoTile
      imageSize=<32 32>
      margin=<ltrb 0 0 0 0>
      outlineSettings {
       bUseBrushTransparency=false
       roundingType=FixedRadius
       width=1
       color {colorUseRule=UseColor_ColorTable, specifiedColor=<rgba 0 0 0 0>}
       cornerRadii=<4 4 4 4>
      }
      tintColor {colorUseRule=UseColor_Specified, specifiedColor=<rgba 0.0284 0.0284 0.0284 1>}
      uVRegion {bIsValid=false, max=<0 0>, min=<0 0>}
     }
     pressedForeground {colorUseRule=UseColor_ColorTable, specifiedColor=<rgba 0 0 0 0>}
     pressedPadding=<ltrb 12 2.5 12 0.5>
     pressedSlateSound {resourceObject=null}
    }
   }
   slot {class=VerticalBoxSlot}
  }
  TextBlock {
   widgetClassPath=&TextBlock
   parent=BtnAssemblySaveToPCGDataAsset
   widget=TextBlock
   properties {
    text=Export Assembly to PCGData Asset
    font {
     bForceMonospaced=false
     bMaterialIsStencil=false
     fontMaterial=null
     fontObject=&Roboto
     letterSpacing=0
     monospacedWidth=1
     size=12
     skewAmount=0
     typefaceFontName=Regular
     outlineSettings {
      bApplyOutlineToDropShadows=false
      bMiteredCorners=false
      bSeparateFillAlpha=false
      outlineMaterial=null
      outlineSize=0
      outlineColor=<rgba 0 0 0 1>
     }
    }
   }
   slot {class=ButtonSlot}
  }
 ]
}
