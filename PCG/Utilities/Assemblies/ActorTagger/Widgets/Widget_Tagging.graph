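# Editor utility widget that adds, removes and selects by actor tags on level selections or, optionally, on Content Browser Blueprints via subobject data, then compiles and saves them.
# Data tables in the widget's own folder populate multi-click buttons, whose mouse button and modifier keys drive MultiClickTagging.
# Key:Value tags support typed input widgets, tri-state values, exclusivity and mirroring into component properties.
# Selection-change events refresh button highlights.
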EditorUtilityWidgetBlueprint /PCG/Utilities/Assemblies/ActorTagger/Widgets/Widget_Tagging
graphs [
 AddTagToSelectedActors {nodes [
   n1 = AddTagToSelectedActors!() -> (then, Transaction:bool=~, ActorTag:String, SubComponents:Array of Strings, ExclusiveTags:Array of Names, NumericTag:bool=~) {is=title K2Node_FunctionEntry}
   > n2 = GetSelectedLevelActors(self=Get EditorActorSubsystem())
   > n3 = For Each Loop(Array:Array of Actor Object References=@n2) -> (Array Element:Actor Object Reference)
   > n4 = Branch(Condition=@n1)
   > n5 = TransactObject(Object=@n3)
   n6 = Branch(execute=[@n5.then, @n4.else], Condition=Get NumericTag() -> (NumericTag:bool))
   n7 = Get Tags!(self:Actor Object Reference=@n3) -> (Tags:Array of Names)
   > n8 = FindPartialTag!(
    execute
    self
    TagsToSearch:Array of Names=@n7
    TagText:String=Get ActorTag() -> (ActorTag:String)
    RemoveFoundTags:bool=true
    AllowTagsWithoutColons:bool=true
   ) -> (then, ReturnedTags:Array of Names, PartialTagFound:bool)
   > n9 = Set Tags!(execute, Tags:Array of Names=@n8, self:Actor Object Reference=@n3) -> (then, Output_Get:Array of Names)
   > n10 = Add Unique(TargetArray:Array of Names=@n9, NewItem:"Name (by ref)"=String To Name(InString=Get ActorTag()))
   n11 = TagExclusivityCheck!(
    execute=[@n10.then, @n21.then]
    self
    TargetTagArray:Array of Strings
    ExclusiveTagsToRemove:Array of Names=Get ExclusiveTags() -> (ExclusiveTags:Array of Names)
    TargetNameTagArray:Array of Names=@n7
    NameArraysOnly:bool=true
   ) -> (then, FilteredTagArray:Array of Strings, TagsWereChanged:bool, FilteredNameTagArray:Array of Names)
   > n12 = Branch(Condition=OR Boolean(A=@n11.TagsWereChanged, B="integer >= integer"(A=@n10)))
   > n13 = Set Tags(Tags=@n11.FilteredNameTagArray, self=@n3)
   n14 = Get ApplyToBPComponents() -> (ApplyToBPComponents:Editor Utility Check Box Object Reference)
   n15 = Length(TargetArray:Array of Strings=Get SubComponents() -> (SubComponents:Array of Strings))
   Branch(execute=[@n13.then, @n12.else], Condition=AND Boolean(A="integer > integer"(A=@n15), B=IsChecked(self=@n14)))
   > n16 = For Each Loop(Array:Array of Static Mesh Component Object References=Get Components By Class!(self:Actor Object Reference=@n3, ComponentClass:Actor Component Class Reference=&StaticMeshComponent) -> (ReturnValue:Array of Static Mesh Component Object References)) -> (Array Element:Static Mesh Component Object Reference)
   n17 = Split(SourceString=Get ActorTag(), InStr=:, SearchDir=FromStart)
   n18 = String To Name(InString=@n17)
   n19 = "To Byte (Integer)"!(InInt:int=String To Integer(InString=@n17.RightS)) -> (ReturnValue:Byte=0)
   > SetEditorProperty(Object=@n16, PropertyName=@n18, PropertyValue:Byte=@n19)
   n20 = For Each Loop(
    Exec=@n16.Completed
    Array:Array of Scene Component Object References=Get Components By Class(self=@n3, ComponentClass=/Script/Engine.SceneComponent) -> (ReturnValue:Array of Scene Component Object References)
   ) -> (Array Element:Scene Component Object Reference)
   > SetEditorProperty(Object=@n20, PropertyName=@n18, PropertyValue=@n19)
   n21 = Add Unique(execute=@n6.else, TargetArray=@n7, NewItem=String To Name(InString=Get ActorTag()))
  ]}
 AssetBrowserTagAdd {nodes [
   AssetBrowserTagAdd!() -> (then, ActorTag:String, SubComponents:Array of Strings, ExclusiveTags:Array of Names, TagType:int) {is=title K2Node_FunctionEntry}
   > n1 = GetSelectedAssets()
   > n2 = For Each Loop(Array:Array of Object References=@n1) -> (Array Element:Object Reference)
   n3 = GetAsset(InAssetData=CreateAssetData(InAsset=@n2, bAllowBlueprintClass=false))
   > n4 = GetBlueprintAsset(Object=@n3)
   > n5 = Is Valid(InputObject=@n4)
   PrintString(execute=@n5.Is Not Valid, InString=Append(A=Get Object Path String(Object=@n3), B=" is not a Valid Blueprint Actor Asset."))
   n6 = Gather Subobject Data For Blueprint(execute=@n5.Is Valid, self=Get SubobjectDataSubsystem(), Context=@n4)
   > n7 = For Each Loop(Array:Array of Subobject Data Handle Structures=@n6) -> (Array Element:Subobject Data Handle Structure)
   n8 = GetAssociatedObject!(Data:"Subobject Data Structure (by ref)"=GetData(DataHandle=@n7)) -> (ReturnValue:Object Reference)
   > n9 = GetEditorProperty!(execute:Exec, Object:Object Reference=@n8, PropertyName:Name=Tags) -> (then:Exec, PropertyValue:Array of Names, ReturnValue:bool)
   > n10 = Set LocalActorTags(LocalActorTags:Array of Names=@n9) -> (Output_Get:Array of Names)
   > n11 = Branch(Condition="integer > integer"(A=Get TagType() -> (TagType:int=0)))
   > n12 = FindPartialTag(TagText=Get ActorTag(), RemoveFoundTags=true, AllowTagsWithoutColons=false)
   > n13 = Set LocalActorTags(LocalActorTags=@n12)
   Add Unique(execute=[@n11.else, @n13.then], TargetArray=@n10, NewItem=String To Name(InString=Get ActorTag()))
   > n14 = TagExclusivityCheck(ExclusiveTagsToRemove=Get ExclusiveTags(), TargetNameTagArray=Get LocalActorTags() -> (LocalActorTags:Array of Names), NameArraysOnly=true)
   > SetEditorProperty(Object=@n8, PropertyName=Tags, PropertyValue:Array of Names=@n14.FilteredNameTagArray)
   > CompileBlueprint!(execute, Blueprint:Blueprint Object Reference=@n4) -> (then, ReturnValue:bool)
   > SaveLoadedAsset(AssetToSave=@n4, bOnlyIfIsDirty=false)
   > n15 = GetBlueprintAsset(Object=@n4)
   n16 = Get ApplyToBPComponents()
   n17 = Length(TargetArray=Get SubComponents())
   > n18 = Branch(Condition=AND Boolean(A="integer > integer"(A=@n17), B=IsChecked(self=@n16)))
   > n19 = Gather Subobject Data For Blueprint(self=Get SubobjectDataSubsystem(), Context=@n15)
   > n20 = For Each Loop(Array=@n19)
   CompileBlueprint(execute=[@n18.else, @n20.Completed], Blueprint=@n15)
   > SaveLoadedAsset(AssetToSave=@n15, bOnlyIfIsDirty=false)
   n21 = GetAssociatedObject(Data=GetData(DataHandle=@n20))
   n22 = Cast To StaticMeshComponent(execute=@n20.LoopBody, Object=@n21)
   n23 = Split(SourceString=Get ActorTag(), InStr=:, SearchDir=FromStart)
   n24 = String To Name(InString=@n23)
   n25 = "To Byte (Integer)"(InInt=String To Integer(InString=@n23.RightS)) -> (ReturnValue=0)
   > SetEditorProperty(Object=@n22, PropertyName=@n24, PropertyValue:Byte=@n25)
   n26 = Cast To SceneComponent!(execute=@n22.CastFailed, Object=@n21) -> (then, CastFailed, AsScene Component:Scene Component Object Reference) {is=title K2Node_DynamicCast}
   > SetEditorProperty(Object=@n26, PropertyName=@n24, PropertyValue=@n25)
  ]}
 AssetBrowserTagRemove {nodes [
   AssetBrowserTagRemove!() -> (then, ActorTag:String, SubComponents:Array of Strings, ExclusiveTags:Array of Names, TagType:int) {is=title K2Node_FunctionEntry}
   > n1 = GetSelectedAssets()
   > n2 = For Each Loop(Array:Array of Object References=@n1) -> (Array Element:Object Reference)
   n3 = GetAsset(InAssetData=CreateAssetData(InAsset=@n2, bAllowBlueprintClass=false))
   > n4 = GetBlueprintAsset(Object=@n3)
   > n5 = Is Valid(InputObject=@n4)
   PrintString(execute=@n5.Is Not Valid, InString=Append(A=Get Object Path String(Object=@n3), B=" is not a Valid Blueprint Actor Asset."))
   n6 = Gather Subobject Data For Blueprint(execute=@n5.Is Valid, self=Get SubobjectDataSubsystem(), Context=@n4)
   > n7 = For Each Loop(Array:Array of Subobject Data Handle Structures=@n6) -> (Array Element:Subobject Data Handle Structure)
   n8 = GetAssociatedObject(Data=GetData(DataHandle=@n7))
   > n9 = GetEditorProperty(Object=@n8, PropertyName=Tags)
   > n10 = Set LocalActorTags(LocalActorTags=@n9)
   > n11 = Branch(Condition="integer > integer"(A=Get TagType() -> (TagType=0)))
   > n12 = FindPartialTag(TagsToSearch=@n10, TagText=Get ActorTag(), RemoveFoundTags=true, AllowTagsWithoutColons=true)
   > n13 = SetEditorProperty(Object=@n8, PropertyName=Tags, PropertyValue:Array of Names=@n12)
   CompileBlueprint(execute=[@n14.then, @n13.then], Blueprint=@n4)
   > SaveLoadedAsset(AssetToSave=@n4, bOnlyIfIsDirty=false)
   Remove Item!(execute=@n11.else, TargetArray:Array of Names=@n10, Item:"Name (by ref)"=String To Name(InString=Get ActorTag())) -> (then, ReturnValue:bool)
   > n14 = SetEditorProperty(Object=@n8, PropertyName=Tags, PropertyValue=@n10)
  ]}
 ButtonClicked_Dispatcher {nodes [ButtonClicked_Dispatcher!() -> (then, Button:Widget Multi Click Button Object Reference) {is=title K2Node_FunctionEntry}]}
 ButtonTagHighlights {nodes [
   ButtonTagHighlights() {is=title K2Node_FunctionEntry}
   n1 = Get ChildWidgets() -> (ChildWidgets:Array of Widget Object References)
   n2 = Length(TargetArray:Array of Widget Object References=@n1)
   > PrintString(
    InString=Append!(A="Processing button highlights for ", B="To String (Integer)"(InInt=@n2), C:String=" buttons") -> (ReturnValue)
    bPrintToScreen=false
    bPrintToLog=Get DebugOutput() -> (DebugOutput:bool)
   )
   > n3 = For Each Loop(Array:Array of Widget Object References=@n1) -> (Array Element:Widget Object Reference)
   Return Node(execute=@n3.Completed)
   > n4 = Cast To Widget_MultiClickButton!(execute, Object=@n3) -> (then, CastFailed, AsWidget Multi Click Button:Widget Multi Click Button Object Reference) {is=title K2Node_DynamicCast}
   > n5 = Branch(Condition=Get TriState!(self:Widget Multi Click Button Object Reference=@n4) -> (TriState:bool))
   n6 = Get TagText!(self:Widget Multi Click Button Object Reference=@n4) -> (TagText:String)
   n7 = "integer > integer"(
    A=Find Item!(
     TargetArray:Array of Names=Get SelectedActorTags() -> (SelectedActorTags:Array of Names)
     ItemToFind:"Name (by ref)"=String To Name(InString=@n6)
    ) -> (ReturnValue)
    B=-1
   )
   > n8 = Branch(Condition=@n7)
   > SetColorAndOpacity(
    self:User Widget Object Reference=@n4
    InColorAndOpacity:Linear Color Structure=Get MultiTag_ButtonTagHighlightMixed() -> (MultiTag_ButtonTagHighlightMixed:Linear Color Structure)
   )
   n9 = String To Name(InString=Append!(A=Split(SourceString=@n6, InStr=:, SearchDir=FromStart), B=:False) -> (ReturnValue))
   n10 = Get SelectedActorTags()
   n11 = Branch(execute=@n8.else, Condition="integer > integer"(A=Find Item(TargetArray=@n10, ItemToFind=@n9), B=-1))
   > SetColorAndOpacity(self=@n4, InColorAndOpacity=Get MultiTag_ButtonTagHighlightOverride() -> (MultiTag_ButtonTagHighlightOverride:Linear Color Structure))
   n12 = Branch(execute=@n11.else, Condition=@n7)
   > SetColorAndOpacity(self=@n4, InColorAndOpacity=Get MultiTag_ButtonTagHighlightColor() -> (MultiTag_ButtonTagHighlightColor:Linear Color Structure))
   SetColorAndOpacity(
    execute=@n12.else
    self=@n4
    InColorAndOpacity=Get MultiTag_ButtonTagNoHighlightColor() -> (MultiTag_ButtonTagNoHighlightColor:Linear Color Structure)
   )
   n13 = FindPartialTag(execute=@n5.else, TagsToSearch=Get SelectedActorTags(), TagText=Get TagText(self=@n4), RemoveFoundTags=false, AllowTagsWithoutColons=true)
   > n14 = Branch(Condition=@n13.PartialTagFound)
   > SetColorAndOpacity(self=@n4, InColorAndOpacity=Get MultiTag_ButtonTagHighlightColor())
   SetColorAndOpacity(execute=@n14.else, self=@n4, InColorAndOpacity=Get MultiTag_ButtonTagNoHighlightColor())
   PrintString(execute=@n4.CastFailed, InString="[Tagger][ButtonTagHighlights] ::Failed to Cast to Multi Widget Button Widget")
  ]}
 EventGraph {nodes [
   "On Selection Changed (ComboBoxTagButtonDefinitions)"!() -> (OutputDelegate, then, SelectedItem:String, SelectionType:ESelectInfo Enum) {is=title K2Node_ComponentBoundEvent}
   n1 = Get ComboBoxTagButtonDefinitions() -> (ComboBoxTagButtonDefinitions:Editor Utility Combo Box String Object Reference)
   > Set UIButtonsDataTable(UIButtonsDataTable:Soft Object Reference="Get (a copy)"(
     Array:Array of Soft Object References=Get UIButtonDataTablesPreviewLevels() -> (UIButtonDataTablesPreviewLevels:Array of Soft Object References)
     Dimension 1=GetSelectedIndex!(self:"Combo Box (String) Object Reference"=@n1) -> (ReturnValue:int=0)
    ) -> (Output:Soft Object Reference)) -> (Output_Get:Soft Object Reference)
   > PopulateUIButtonsFromDataTable!(execute, self, DataTable:Soft Object Reference=Get UIButtonsDataTable() -> (UIButtonsDataTable:Soft Object Reference)) -> (then)
   n2 = RegisterButton!() -> (OutputDelegate, then, Button:Widget Multi Click Button Object Reference) {is=title K2Node_CustomEvent}
   n3 = ButtonClicked_Dispatcher_Event!() -> (OutputDelegate, then, Button:Widget Multi Click Button Object Reference) {is=title K2Node_CustomEvent}
   Bind Event to Btn Event Dispatch(execute=@n2.then, self:Widget Multi Click Button Object Reference=@n2.Button, Delegate=@n3) {is=title K2Node_AddDelegate}
   n4 = UpdateTagButtonUI!() -> (OutputDelegate, then, SelectionSet:Typed Element Selection Set Object Reference) {is=title K2Node_CustomEvent}
   > n5 = Set TypedElementSelectionSet(TypedElementSelectionSet:Typed Element Selection Set Object Reference=@n4.SelectionSet) -> (Output_Get:Typed Element Selection Set Object Reference)
   > UpdatePredefinedTagButtonHighlights!(
    execute
    self
    TypedElementSelectionSet:Typed Element Selection Set Object Reference=@n5
    DebugOutputGroup:String="[UpdateTagButtonUI custom Event]"
   ) -> (then)
   n6 = UpdateTypedElementSelectionSetReference!() -> (OutputDelegate, then, MapChangeEventFlags:int) {is=title K2Node_CustomEvent}
   > n7 = GetSelectionSet(self=Get LevelEditorSubsystem())
   > n8 = Set TypedElementSelectionSet(TypedElementSelectionSet=@n7)
   > PrintString(InString=GetDisplayName(Object=@n8), bPrintToScreen=false, bPrintToLog=Get DebugOutput())
   n9 = Event PreConstruct()
   > Branch(Condition=@n9.IsDesignTime)
   > PopulateInternalAssets()
   > PopulateUIButtonsFromDataTable()
   Event Construct()
   > PopulateInternalAssets()
   > PopulateUIButtonsFromDataTable()
   > n10 = GetSelectionSet(self=Get LevelEditorSubsystem())
   > n11 = Set TypedElementSelectionSet(TypedElementSelectionSet=@n10)
   > Unbind Event from On Selection Change(self:Typed Element Selection Set Object Reference=@n11, Delegate=@n4) {is=title K2Node_RemoveDelegate}
   > Bind Event to On Selection Change(self:Typed Element Selection Set Object Reference=@n11, Delegate=@n4) {is=title K2Node_AddDelegate}
   n12 = Get LevelEditorSubsystem()
   > Unbind Event from On Map Changed(self:Level Editor Subsystem Object Reference=@n12, Delegate=@n6) {is=title K2Node_RemoveDelegate}
   > Bind Event to On Map Changed(self:Level Editor Subsystem Object Reference=@n12, Delegate=@n6) {is=title K2Node_AddDelegate}
   Event OnMouseEnter()
   > SetKeyboardFocus(self:Widget Object Reference)
   Event OnMouseLeave!() -> (OutputDelegate, then, MouseEvent:"Pointer Event Structure (by ref)") {is=title K2Node_Event}
   > Set CtrlPressed(CtrlPressed:bool=false) -> (Output_Get:bool)
   > Set ShiftPressed(ShiftPressed:bool=false) -> (Output_Get:bool)
   MultiClickTagging!(execute=@n3.then, self, Button:Widget Multi Click Button Object Reference=@n3.Button, KeyboardModifiers:int=0) -> (then)
  ]}
 FindPartialTag {nodes [
   FindPartialTag!() -> (then, TagsToSearch:Array of Names, TagText:String, RemoveFoundTags:bool=true, AllowTagsWithoutColons:bool=true) {is=title K2Node_FunctionEntry}
   > Set LocalPartialTagFound(LocalPartialTagFound:bool=false) -> (Output_Get:bool)
   > n1 = For Each Loop(Array:Array of Names=Get TagsToSearch() -> (TagsToSearch:Array of Names)) -> (Array Element:Name)
   Remove Item(execute=@n1.Completed, TargetArray=Get TagsToSearch(), Item=REMOVED)
   > Return Node!(execute, ReturnedTags:Array of Names=Get TagsToSearch(), PartialTagFound:bool=Get LocalPartialTagFound() -> (LocalPartialTagFound:bool))
   n2 = FindSubstring(
    SearchIn="To String (Name)"(InName=@n1)
    Substring=Split(SourceString=Get TagText!() -> (TagText), InStr=:, SearchDir=FromStart)
    bUseCase=true
   )
   n3 = Branch(execute=@n1.LoopBody, Condition=AND Boolean(A="Equal (Integer)"(A=@n2), B=Get RemoveFoundTags() -> (RemoveFoundTags:bool)))
   > Set LocalPartialTagFound(LocalPartialTagFound=true)
   > Set Array Elem!(execute, TargetArray:Array of Names=Get TagsToSearch(), Index:int=@n1.Array Index, Item:"Name (by ref)"=REMOVED, bSizeToFit:bool=false) -> (then)
   n4 = FindSubstring(SearchIn="To String (Name)"(InName=@n1), Substring=Get TagText())
   Branch(execute=@n3.else, Condition=AND Boolean(A="Equal (Integer)"(A=@n4), B=Get AllowTagsWithoutColons() -> (AllowTagsWithoutColons:bool)))
   > Set LocalPartialTagFound(LocalPartialTagFound=true)
  ]}
 MultiClickTagging {nodes [
   n1 = EndTransaction()
   UpdatePredefinedTagButtonHighlights(
    execute=[@n38.then, @n51.then, @n50.then, @n40.Completed, @n10.then, @n7.then, @n17.else, @n18.Completed, @n1.then, @n64.then, @n52.then]
    TypedElementSelectionSet=Get TypedElementSelectionSet() -> (TypedElementSelectionSet:Typed Element Selection Set Object Reference)
    DebugOutputGroup="[MultiClickTagging]"
   )
   > Return Node!(execute)
   n2 = MultiClickTagging!() -> (then, Button:Widget Multi Click Button Object Reference, KeyboardModifiers:int) {is=title K2Node_FunctionEntry}
   > Set PressedButton(PressedButton:Widget Multi Click Button Object Reference=@n2) -> (Output_Get:Widget Multi Click Button Object Reference)
   > Set Exclusive(Exclusive:Array of Names=Get Exclusive!(self:Widget Multi Click Button Object Reference=Get PressedButton() -> (PressedButton:Widget Multi Click Button Object Reference)) -> (Exclusive:Array of Names)) -> (Output_Get:Array of Names)
   > Set PostTagFunction(PostTagFunction:int=Get PostTagFunction!(self:Widget Multi Click Button Object Reference=Get PressedButton()) -> (PostTagFunction:int=0)) -> (Output_Get:int=0)
   > Set TriStateButton(TriStateButton:bool=Get TriState(self=Get PressedButton())) -> (Output_Get:bool)
   > Set MouseButtonState(MouseButtonState:int=Get MouseButtonState!(self:Widget Multi Click Button Object Reference=Get PressedButton()) -> (MouseButtonState:int=0)) -> (Output_Get:int=0)
   > Set Subcomponents(Subcomponents:Array of Strings=Get Subcomponents!(self:Widget Multi Click Button Object Reference=Get PressedButton()) -> (Subcomponents:Array of Strings)) -> (Output_Get:Array of Strings)
   > Set ButtonTagText(ButtonTagText:String=Get TagText!(self:Widget Multi Click Button Object Reference=Get PressedButton()) -> (TagText)) -> (Output_Get:String)
   > n3 = Branch(Condition="integer <= integer"(A=Get MouseButtonState!() -> (MouseButtonState=0), B=4))
   > n4 = Switch on Int!(execute:Exec, Selection:int=Get MouseButtonState() -> (MouseButtonState=0)) -> (Default:Exec, 0:Exec, 1:Exec, 2:Exec, 3:Exec, 4:Exec)
   Set AddTagName(execute=@n4.2, AddTagName:Name=String To Name(InString=Get ButtonTagText() -> (ButtonTagText:String))) -> (Output_Get:Name=None)
   > n5 = Branch(Condition=IsChecked(self=Get CheckBoxContentBrowserTagging() -> (CheckBoxContentBrowserTagging:Editor Utility Check Box Object Reference)))
   > n6 = Switch on Int!(execute, Selection=Get PostTagFunction!() -> (PostTagFunction=0)) -> (Default, 0, 1, 2, 3, 4, 5:Exec)
   n7 = AssetBrowserTagAdd!(
    execute=[@n6.0, @n6.Default]
    self
    ActorTag:String=Get ButtonTagText()
    SubComponents:Array of Strings=Get Subcomponents!() -> (Subcomponents)
    ExclusiveTags:Array of Names=Get Exclusive!() -> (Exclusive)
    TagType:int=Get PostTagFunction() -> (PostTagFunction=0)
   ) -> (then) {is=title K2Node_CallFunction}
   n8 = Cast To Widget_InputFloat!(
    execute=@n6.1
    Object=GetWidgetAtIndex!(
     self:Widget Switcher Object Reference=Get InputWidgetSwitcher() -> (InputWidgetSwitcher:Widget Switcher Object Reference)
     Index:int=Get PostTagFunction() -> (PostTagFunction=0)
    ) -> (ReturnValue:Widget Object Reference)
   ) -> (then, CastFailed, AsWidget Input Float:Widget Input Float Object Reference) {is=title K2Node_DynamicCast}
   n9 = Get ButtonTagText()
   > n10 = AssetBrowserTagAdd(
    ActorTag=Append!(
     A=Split(SourceString=@n9, InStr=:, SearchDir=FromStart)
     B=:
     C:String=Get InputValue!(self:Widget Input Float Object Reference=@n8) -> (InputValue:String)
    ) -> (ReturnValue)
    SubComponents=Get Subcomponents()
    ExclusiveTags=Get Exclusive()
    TagType=Get PostTagFunction() -> (PostTagFunction=0)
   )
   n11 = Branch(execute=@n5.else, Condition=Get TriStateButton() -> (TriStateButton:bool))
   > n12 = Branch(Condition=Get CtrlPressed() -> (CtrlPressed:bool))
   n13 = Split(SourceString=Get ButtonTagText(), InStr=:, SearchDir=FromStart)
   > n14 = Branch(Condition=@n13.ReturnValue)
   > AddTagToSelectedActors!(
    execute
    self
    Transaction:bool=true
    ActorTag:String=Append!(A=@n13, B=:False) -> (ReturnValue)
    SubComponents:Array of Strings
    ExclusiveTags:Array of Names
    NumericTag:bool=false
   ) -> (then) {is=title K2Node_CallFunction}
   > n15 = RemoveTagFromActors!(
    execute
    self
    Transaction:bool=false
    ActorTagText:String=Get ButtonTagText()
    UseSelectedActors:bool=true
    ActorArray:Array of Actor Object References
    SubComponents:Array of Strings
    ExclusiveTags:Array of Names
    AllowForTagsHavingNumericData:bool=false
   ) -> (then)
   n16 = Get Exclusive()
   n17 = Branch(
    execute=[@n15.then, @n24.then, @n30.then, @n36.then, @n33.then, @n27.Completed, @n19.then, @n21.then]
    Condition="integer > integer"(A=Length(TargetArray:Array of Names=@n16))
   )
   > n18 = For Each Loop(Array=Get Exclusive())
   > RemoveTagFromActors(Transaction=true, ActorTagText="To String (Name)"(InName=@n18), UseSelectedActors=true, AllowForTagsHavingNumericData=false)
   PrintString(execute=@n14.else, InString=Tri State but no colon in tag text)
   n19 = AddTagToSelectedActors(execute=@n12.else, Transaction=true, ActorTag=Get ButtonTagText(), NumericTag=false)
   n20 = Switch on Int(execute=@n11.else, Selection=Get PostTagFunction() -> (PostTagFunction=0))
   n21 = AddTagToSelectedActors(
    execute=[@n20.0, @n20.Default]
    Transaction=true
    ActorTag=Get ButtonTagText()
    SubComponents=Get Subcomponents()
    ExclusiveTags=Get Exclusive()
    NumericTag=false
   )
   n22 = Cast To Widget_InputFloat(execute=@n20.1, Object=GetWidgetAtIndex(self=Get InputWidgetSwitcher(), Index=Get PostTagFunction() -> (PostTagFunction=0)))
   > n23 = FormatTagValue!(execute, self:Widget Input Float Object Reference=@n22, Tag:String=Get ButtonTagText(), ActorObjectReference:Actor Object Reference) -> (then, TagResult:String)
   > n24 = AddTagToSelectedActors(Transaction=false, ActorTag=@n23, SubComponents=Get Subcomponents(), ExclusiveTags=Get Exclusive(), NumericTag=true)
   PrintString(execute=@n22.CastFailed, InString=Widget_InputFloat Cast failed)
   n25 = Cast To Widget_InputClassDefaultObject!(execute=@n20.2, Object=GetWidgetAtIndex(self=Get InputWidgetSwitcher(), Index=Get PostTagFunction() -> (PostTagFunction=0))) -> (then, CastFailed, AsWidget Input Class Default Object:Widget Input Class Default Object Object Reference) {is=title K2Node_DynamicCast}
   > n26 = GetSelectedLevelActors(self=Get EditorActorSubsystem())
   > n27 = For Each Loop(Array:Array of Actor Object References=@n26) -> (Array Element:Actor Object Reference)
   > FormatTagValue(self:Widget Input Class Default Object Object Reference=@n25, Tag=Get ButtonTagText(), ActorObjectReference=@n27)
   PrintString(execute=@n25.CastFailed, InString=Widget_InputCDO Cast failed.)
   n28 = Cast To Widget_InputSoftObjectPath!(execute=@n20.3, Object=GetWidgetAtIndex(self=Get InputWidgetSwitcher(), Index=Get PostTagFunction() -> (PostTagFunction=0))) -> (then, CastFailed, AsWidget Input Soft Object Path:Widget Input Soft Object Path Object Reference) {is=title K2Node_DynamicCast}
   > n29 = FormatTagValue(self:Widget Input Soft Object Path Object Reference=@n28, Tag=Get ButtonTagText())
   > n30 = AddTagToSelectedActors(Transaction=false, ActorTag=@n29, SubComponents=Get Subcomponents(), ExclusiveTags=Get Exclusive(), NumericTag=true)
   PrintString(execute=@n28.CastFailed, InString=Widget_InputSoftObjectPath Cast failed.)
   n31 = Cast To Widget_InputString!(execute=@n20.4, Object=GetWidgetAtIndex(self=Get InputWidgetSwitcher(), Index=Get PostTagFunction() -> (PostTagFunction=0))) -> (then, CastFailed, AsWidget Input String:Widget Input String Object Reference) {is=title K2Node_DynamicCast}
   > n32 = FormatTagValue(self:Widget Input String Object Reference=@n31, Tag=Get ButtonTagText())
   > n33 = AddTagToSelectedActors(Transaction=false, ActorTag=@n32, SubComponents=Get Subcomponents(), ExclusiveTags=Get Exclusive(), NumericTag=true)
   PrintString(execute=@n31.CastFailed, InString=Widget_InputString Cast failed.)
   n34 = Cast To Widget_InputVector3!(execute=@n20.5, Object=GetWidgetAtIndex(self=Get InputWidgetSwitcher(), Index=Get PostTagFunction() -> (PostTagFunction=0))) -> (then, CastFailed, AsWidget Input Vector 3:Widget Input Vector 3 Object Reference) {is=title K2Node_DynamicCast}
   > n35 = FormatTagValue(self:Widget Input Vector 3 Object Reference=@n34, Tag=Get ButtonTagText())
   > n36 = AddTagToSelectedActors(Transaction=false, ActorTag=@n35, SubComponents=Get Subcomponents(), ExclusiveTags=Get Exclusive(), NumericTag=true)
   PrintString(execute=@n34.CastFailed, InString=Widget_InputVector3 Cast failed.)
   n37 = Branch(execute=@n4.3, Condition=IsChecked(self=Get CheckBoxContentBrowserTagging()))
   > n38 = AssetBrowserTagRemove!(
    execute
    self
    ActorTag:String=Get ButtonTagText()
    SubComponents:Array of Strings=Get Subcomponents()
    ExclusiveTags:Array of Names=Get Exclusive()
    TagType:int=Get PostTagFunction() -> (PostTagFunction=0)
   ) -> (then) {is=title K2Node_CallFunction}
   n39 = GetSelectedLevelActors(execute=@n37.else, self=Get EditorActorSubsystem())
   > n40 = For Each Loop(Array=@n39)
   > Set CurrentActor(CurrentActor:Actor Object Reference=@n40) -> (Output_Get:Actor Object Reference)
   > n41 = Branch(Condition=Get TriStateButton())
   n42 = Split(SourceString=Get ButtonTagText(), InStr=:, SearchDir=FromStart)
   n43 = Append(A=@n42, B=:False)
   n44 = String To Name(InString=@n43)
   n45 = Get Tags(self=Get CurrentActor() -> (CurrentActor:Actor Object Reference))
   > n46 = Branch(Condition="integer >= integer"(A=Find Item(TargetArray=@n45, ItemToFind=@n44)))
   > n47 = RemoveTagFromActors(Transaction=true, ActorTagText=@n43, UseSelectedActors=true, AllowForTagsHavingNumericData=false)
   n48 = String To Name(InString=Append(A=@n42, B=:True))
   n49 = Branch(execute=[@n47.then, @n46.else], Condition="integer >= integer"(A=Find Item(TargetArray=@n45, ItemToFind=@n48)))
   > RemoveTagFromActors(Transaction=true, ActorTagText=Get ButtonTagText(), UseSelectedActors=true, AllowForTagsHavingNumericData=false)
   RemoveTagFromActors(
    execute=@n49.else
    Transaction=true
    ActorTagText=Split(SourceString=Get ButtonTagText(), InStr=:, SearchDir=FromStart)
    UseSelectedActors=true
    AllowForTagsHavingNumericData=false
   )
   RemoveTagFromActors(
    execute=@n41.else
    Transaction=true
    ActorTagText=Get ButtonTagText()
    UseSelectedActors=true
    SubComponents=Get Subcomponents()
    ExclusiveTags=Get Exclusive()
    AllowForTagsHavingNumericData=true
   )
   Branch(execute=@n4.4, Condition="integer > integer"(A=Get PostTagFunction() -> (PostTagFunction=0)))
   > n50 = Branch(Condition=Get AltPressed() -> (AltPressed:bool))
   n51 = SetActiveWidgetIndex(execute=@n50.else, self=Get InputWidgetSwitcher(), Index=Get PostTagFunction() -> (PostTagFunction=0))
   n52 = Branch(execute=@n4.1, Condition=IsChecked(self=Get CheckBoxContentBrowserTagging()))
   n53 = Branch(execute=@n52.else, Condition=Get TriStateButton())
   > n54 = Branch(Condition=Get CtrlPressed())
   n55 = Get ButtonTagText()
   > n56 = Set TargetTagText(TargetTagText:String=Append(A=Split(SourceString=@n55, InStr=:, SearchDir=FromStart), B=:False)) -> (Output_Get:String)
   n57 = Branch(
    execute=[@n56.then, @n76.then]
    Condition=OR Boolean!(A=Get AltPressed(), B=Get ShiftPressed() -> (ShiftPressed:bool), C:bool) -> (ReturnValue)
   )
   > n58 = GetSelectedLevelActors(self=Get EditorActorSubsystem())
   > Set NewActorSelection(NewActorSelection:Array of Actor Object References=@n58) -> (Output_Get:Array of Actor Object References)
   > GetEditorWorld(self=Get UnrealEditorSubsystem())
   > n59 = GetAllActorsWithTag!(execute, WorldContextObject:Object Reference, Tag:Name=String To Name(InString=Get TargetTagText() -> (TargetTagText:String))) -> (then, OutActors:Array of Actor Object References)
   > n60 = Branch(Condition=Get ShiftPressed())
   > Append Array!(
    execute
    TargetArray:Array of Actor Object References=Get NewActorSelection() -> (NewActorSelection:Array of Actor Object References)
    SourceArray:Array of Actor Object References=@n59
   ) -> (then)
   n61 = Split(SourceString=Get TargetTagText(), InStr=:, SearchDir=FromStart)
   > n62 = GetAllActorsWithTag(Tag=String To Name(InString=Append(A=@n61, B=:False)))
   > n63 = Append Array(TargetArray=Get NewActorSelection(), SourceArray=@n62)
   n64 = SetSelectedLevelActors(
    execute=[@n63.then, @n75.then, @n65.Completed, @n72.else, @n70.then]
    self=Get EditorActorSubsystem()
    ActorsToSelect=Get NewActorSelection()
   )
   Branch(execute=@n60.else, Condition=Get AltPressed())
   > n65 = For Each Loop(Array=@n59)
   n66 = Get NewActorSelection()
   > Branch(Condition="integer > integer"(A=Find Item(TargetArray:Array of Actor Object References=@n66, ItemToFind:"Actor Object Reference (by ref)"=@n65), B=-1))
   > Remove Item(TargetArray:Array of Actor Object References=Get NewActorSelection(), Item:"Actor Object Reference (by ref)"=@n65)
   n67 = Split(SourceString=Get TargetTagText(), InStr=:, SearchDir=FromStart)
   n68 = Branch(execute=@n57.else, Condition=@n67.ReturnValue)
   > n69 = SelectActorsUsingTagSubstring!(execute, self, TagSubstring:String=Append(A=@n67, B=:)) -> (then, ActorSelection:Array of Actor Object References)
   > n70 = Set NewActorSelection(NewActorSelection=@n69)
   n71 = SelectActorsbyTag!(execute=@n68.else, self, TextTag:String=Get TargetTagText()) -> (then, SelectedActors:Array of Actor Object References, NumSelectedActors:int=0)
   > Set NewActorSelection(NewActorSelection=@n71)
   > n72 = Branch(Condition=Get TriStateButton())
   n73 = Get TargetTagText()
   > n74 = SelectActorsbyTag(TextTag=Append(A=Split(SourceString=@n73, InStr=:, SearchDir=FromStart), B=:False)) -> (NumSelectedActors=0)
   > n75 = Append Array(TargetArray=Get NewActorSelection(), SourceArray=@n74)
   n76 = Set TargetTagText(execute=[@n53.else, @n54.else], TargetTagText=Get ButtonTagText())
   PrintString(execute=@n3.else, InString=Mouse Button State not supported, bPrintToScreen=false, bPrintToLog=Get DebugOutput())
   > Return Node()
   PrintString(InString="To String (Integer)"(InInt=@n1))
  ]}
 OnKeyDown {nodes [
   n1 = OnKeyDown!() -> (then, MyGeometry:Geometry Structure, InKeyEvent:Key Event Structure) {is=title K2Node_FunctionEntry}
   n2 = GetKey!(Input:"Key Event Structure (by ref)"=@n1.InKeyEvent) -> (ReturnValue:Key Structure)
   > n3 = Branch(Condition="Equal (Key)"(A:Key Structure=@n2, B:Key Structure=Escape) -> (ReturnValue:bool=~) {is=title K2Node_PromotableOperator})
   > SetSelectedLevelActors(self=Get EditorActorSubsystem(), ActorsToSelect=Make Array("[0]":Actor Object Reference) -> (Array:Array of Actor Object References))
   > Return Node!(execute, ReturnValue:Event Reply Structure=Handled())
   Set CtrlPressed(
    execute=@n3.else
    CtrlPressed=OR Boolean!(A="Equal (Key)"(A=@n2, B=LeftControl) -> (ReturnValue=~), B="Equal (Key)"(A=@n2, B=RightControl) -> (ReturnValue=~)) -> (ReturnValue)
   )
   > Set AltPressed(AltPressed:bool=OR Boolean(A="Equal (Key)"(A=@n2, B=LeftAlt) -> (ReturnValue=~), B="Equal (Key)"(A=@n2, B=RightAlt) -> (ReturnValue=~))) -> (Output_Get:bool)
   > Set ShiftPressed(ShiftPressed=OR Boolean(A="Equal (Key)"(A=@n2, B=LeftShift) -> (ReturnValue=~), B="Equal (Key)"(A=@n2, B=RightShift) -> (ReturnValue=~)))
   > Return Node(ReturnValue=Handled())
  ]}
 OnKeyUp {nodes [
   n1 = OnKeyUp!() -> (then, MyGeometry:Geometry Structure, InKeyEvent:Key Event Structure) {is=title K2Node_FunctionEntry}
   n2 = GetKey(Input=@n1.InKeyEvent)
   > n3 = Branch(Condition="Equal (Key)"(A=@n2, B=LeftControl) -> (ReturnValue=~))
   > n4 = Set CtrlPressed(CtrlPressed=false)
   n5 = Branch(execute=[@n3.else, @n4.then], Condition="Equal (Key)"(A=@n2, B=LeftAlt) -> (ReturnValue=~))
   > n6 = Set AltPressed(AltPressed=false)
   n7 = Branch(execute=[@n5.else, @n6.then], Condition="Equal (Key)"(A=@n2, B=LeftShift) -> (ReturnValue=~))
   > n8 = Set ShiftPressed(ShiftPressed=false)
   Return Node(execute=[@n7.else, @n8.then], ReturnValue=Handled())
  ]}
 OnMouseButtonDown {nodes [
   n1 = OnMouseButtonDown!() -> (then, MyGeometry:Geometry Structure, MouseEvent:"Pointer Event Structure (by ref)") {is=title K2Node_FunctionEntry}
   n2 = Get Effecting Button!(Input:"Pointer Event Structure (by ref)"=@n1.MouseEvent) -> (ReturnValue:Key Structure) {is=title K2Node_CallFunction}
   > n3 = Branch(Condition="Equal (Key)"!(A:Key Structure=@n2, B:Key Structure=LeftMouseButton) -> (ReturnValue:bool) {is=title K2Node_CallFunction})
   > n4 = Set MouseButtonState(MouseButtonState=1) -> (Output_Get=0)
   Return Node(execute=[@n8.then, @n9.then, @n4.then, @n6.then], ReturnValue=Handled())
   n5 = Branch(execute=@n3.else, Condition="Equal (Key)"(A=@n2, B=RightMouseButton))
   > n6 = Set MouseButtonState(MouseButtonState=2) -> (Output_Get=0)
   n7 = Branch(execute=@n5.else, Condition="Equal (Key)"(A=@n2, B=MiddleMouseButton))
   > n8 = Set MouseButtonState(MouseButtonState=3) -> (Output_Get=0)
   n9 = Set MouseButtonState(execute=@n7.else, MouseButtonState=0) -> (Output_Get=0)
  ]}
 PopulateDropDownfromFolder {nodes [
   n1 = PopulateDropDownfromFolder!() -> (then, ComboBoxString:"Combo Box (String) Object Reference", FolderPath:String) {is=title K2Node_FunctionEntry}
   > Clear(TargetArray:Array of Asset Data Structures=Get DependancyAssetData() -> (DependancyAssetData:Array of Asset Data Structures))
   > PrintString(InString=Populate Drop Down, bPrintToScreen=false, bPrintToLog=Get DebugOutput())
   > Clear(TargetArray:Array of Strings=Get Local_ObjectReferences() -> (Local_ObjectReferences:Array of Strings))
   > n2 = GetAssetsByPath!(
    execute:Exec
    self:Object Reference="To Object (Interface)"!(Interface:"Interface Interface (by ref)"=GetAssetRegistry()) -> (ReturnValue:Object Reference)
    PackagePath:Name=String To Name(InString=@n1.FolderPath)
    bRecursive:bool=false
    bIncludeOnlyOnDiskAssets:bool=true
   ) -> (then:Exec, OutAssetData:Array of Asset Data Structures, ReturnValue:bool) {is=title K2Node_Message}
   > n3 = Set DependancyAssetData(DependancyAssetData:Array of Asset Data Structures=@n2) -> (Output_Get:Array of Asset Data Structures)
   > n4 = For Each Loop(Array:Array of Asset Data Structures=@n3) -> (Array Element:Asset Data Structure)
   SortByAssetName!(execute=@n4.Completed, Assets:Array of Asset Data Structures=@n3, SortOrder:EAssetRegistrySortOrder Enum=Ascending) -> (then)
   > n5 = For Each Loop(Array=Get DependancyAssetData())
   n6 = Get Local_ObjectReferences()
   SortStringArray!(execute=@n5.Completed, TargetArray:Array of Strings=@n6, bStableSort:bool=false, SortOrder:EArraySortOrder Enum=Ascending) -> (then)
   > n7 = For Each Loop(Array:Array of Strings=@n6) -> (Array Element:String)
   n8 = Get OptionMapping() -> (OptionMapping:Map of Strings to Asset Data Structures)
   n9 = Keys!(execute=@n7.Completed, TargetMap:Map of Strings to Asset Data Structures=@n8) -> (then, Keys:Array of Strings)
   > n10 = Values!(execute, TargetMap:Map of Strings to Asset Data Structures=@n8) -> (then, Values:Array of Asset Data Structures)
   > n11 = For Each Loop(Array=@n9)
   PrintString(execute=@n11.Completed, InString="\r\n\r\n", bPrintToScreen=false, bPrintToLog=Get DebugOutput())
   > n12 = For Each Loop(Array:Array of Asset Data Structures=Get DependancyAssetData()) -> (Array Element:Asset Data Structure)
   ClearOptions(execute=@n12.Completed, self:"Combo Box (String) Object Reference"=Get ComboBoxTagButtonDefinitions())
   > Clear(TargetArray:Array of Soft Object References=Get UIButtonDataTablesPreviewLevels())
   > n13 = Values(TargetMap=Get OptionMapping())
   > n14 = For Each Loop(Array=@n13)
   n15 = GetClass!(InAssetData:"Asset Data Structure (by ref)"=@n14) -> (ReturnValue:Object Class Reference)
   > Branch(Condition="Equal, Case Insensitive (String)"(A=Get Class Display Name(Class=@n15), B=DataTable))
   n16 = Break Asset Data(AssetData=@n14)
   n17 = Get ComboBoxTagButtonDefinitions()
   > AddOption!(execute, self:"Combo Box (String) Object Reference"=@n17, Option:String="To String (Name)"(InName=@n16.AssetName)) -> (then)
   n18 = "To String (Name)"(InName=@n16)
   > n19 = Add Unique(
    TargetArray:Array of Soft Object References=Get UIButtonDataTablesPreviewLevels()
    NewItem:"Soft Object Reference (by ref)"=To Soft Object Reference(SoftObjectPath=MakeSoftObjectPath!(PathString:String=@n18) -> (ReturnValue:Soft Object Path Structure))
   )
   RefreshOptions(execute=[@n19.then, @n14.Completed], self:"Combo Box (String) Object Reference"=@n17)
   > SetSelectedIndex!(execute, self:"Combo Box (String) Object Reference"=@n17, Index:int=0) -> (then)
   n20 = Break Asset Data(AssetData=@n12)
   PrintString(
    execute=@n12.LoopBody
    InString=Append!(A="To String (Integer)"(InInt=@n12.Array Index), B=":: ", C:String="To String (Name)"(InName=@n20)) -> (ReturnValue)
    bPrintToScreen=false
    bPrintToLog=Get DebugOutput()
   )
   PrintString(
    execute=@n11.LoopBody
    InString=Append(A="To String (Integer)"(InInt=@n11.Array Index), B=::, C=@n11)
    bPrintToScreen=false
    bPrintToLog=Get DebugOutput()
   )
   n21 = "Get (a copy)"(Array:Array of Asset Data Structures=@n10, Dimension 1=@n11.Array Index) -> (Output:Asset Data Structure)
   > PrintString(InString="To String (Name)"(InName=Break Asset Data(AssetData=@n21).AssetName), bPrintToScreen=false, bPrintToLog=Get DebugOutput())
   n22 = For Each Loop(Exec=@n7.LoopBody, Array=Get DependancyAssetData())
   n23 = Break Asset Data(AssetData=@n22)
   > Branch(Condition="Equal Exactly (String)"(A="To String (Name)"(InName=@n23.AssetName), B=@n7))
   > Add!(
    execute
    TargetMap:Map of Strings to Asset Data Structures=Get OptionMapping()
    Key:"String (by ref)"=@n7
    Value:"Asset Data Structure (by ref)"=@n22
   ) -> (then) {is=title K2Node_CallFunction}
   n24 = GetClass(InAssetData=@n5)
   Branch(execute=@n5.LoopBody, Condition="Equal, Case Insensitive (String)"(A=Get Class Display Name(Class=@n24), B=DataTable))
   > Add(TargetArray=Get Local_ObjectReferences(), NewItem="To String (Name)"(InName=Break Asset Data(AssetData=@n5).AssetName)) {is=title K2Node_CallArrayFunction}
   n25 = Get Class Display Name(Class=GetClass(InAssetData=@n4))
   n26 = Break Asset Data(AssetData=@n4)
   PrintString(
    execute=@n4.LoopBody
    InString=Append(A="To String (Name)"(InName=@n26.AssetName), B=@n25, C=@n25)
    bPrintToScreen=false
    bPrintToLog=Get DebugOutput()
   )
   > n27 = Branch(Condition="Equal Exactly (String)"(A=@n25, B=DataTable))
   Remove Index!(execute=@n27.else, TargetArray:Array of Asset Data Structures=@n3, IndexToRemove:int=@n4.Array Index) -> (then)
  ]}
 PopulateInternalAssets {nodes [
   PopulateInternalAssets() {is=title K2Node_FunctionEntry}
   n1 = GetSoftClassPath!(Class:Object Class Reference=Get Class(Object=Self-Reference()) -> (ReturnValue:Widget Tagging Class Reference)) -> (ReturnValue:Soft Class Path Structure)
   > Set WidgetFolderPath(WidgetFolderPath:String=Split(SourceString=BreakSoftClassPath!(InSoftClassPath:Soft Class Path Structure=@n1) -> (PathString:String), InStr=/, SearchDir=FromEnd)) -> (Output_Get:String)
   > PopulateDropDownfromFolder!(
    execute
    self
    ComboBoxString:"Combo Box (String) Object Reference"=Get ComboBoxTagButtonDefinitions()
    FolderPath:String=Get WidgetFolderPath() -> (WidgetFolderPath:String)
   ) -> (then) {is=title K2Node_CallFunction}
  ]}
 PopulateUIButtonsFromDataTable {nodes [
   PopulateUIButtonsFromDataTable!() -> (then, DataTable:Soft Object Reference) {is=title K2Node_FunctionEntry}
   > ClearChildren(self=Get GPanelDynamicUIButtons() -> (GPanelDynamicUIButtons:Grid Panel Object Reference))
   > Clear(TargetArray:Array of Widget Object References=Get ChildWidgets())
   > n1 = LoadAsset(AssetPath="To String (SoftObjectReference)"!(SoftObjectReference:"Soft Object Reference (by ref)"=Get UIButtonsDataTable()) -> (ReturnValue:String))
   > n2 = Cast To DataTable!(execute, Object=@n1) -> (then, CastFailed, AsData Table:Data Table Object Reference) {is=title K2Node_DynamicCast}
   > n3 = GetDataTableRowNames(Table=@n2)
   > n4 = For Each Loop(Array:Array of Names=@n3) -> (Array Element:Name)
   > n5 = Get Data Table Row!(execute:Exec, DataTable:Data Table Object Reference=@n2, RowName:Name=@n4) -> (then:Exec, RowNotFound:Exec, ReturnValue:S Widget Tagging UIButtons Structure)
   > Set CurrentButton(CurrentButton:Widget Multi Click Button Object Reference) -> (Output_Get:Widget Multi Click Button Object Reference)
   n6 = Break S Widget Tagging UIButtons!(S_WidgetTaggingUIButtons:"S Widget Tagging UIButtons Structure (by ref)"=@n5) -> (
    ButtonText_11_#1:String
    Tag_5_#2:String
    Exclusive_39_#3:Array of Names
    TriState_32_#4:bool=~
    RespectGridRow_57_#5:bool=~
    RespectGridColumn_59_#6:bool=~
    GridPanelColumn_20_#7:int
    GridPanelRow_21_#8:int
    ButtonIsDisabled_27_#9:bool=~
    AdditionalTagFunction_35_#10:int
    ButtonTooltip_61_#11:String
    InputWidget_49_#12:Editor Utility Widget Blueprint Soft Object Reference
    SubComponent_53_#13:Array of Strings
   ) {is=title K2Node_BreakStruct}
   > Branch(Condition="Not Equal Exactly (String)"(A=@n6.Tag_5_#2))
   > n7 = Create Widget!(
    execute:Exec
    WorldContextObject:Object Reference
    Class:User Widget Class Reference=&/Widget_MultiClickButton.Widget_MultiClickButton_C
    OwningPlayer:Player Controller Object Reference
   ) -> (then:Exec, ReturnValue:Widget Multi Click Button Object Reference)
   > Set CurrentButton(CurrentButton=@n7)
   > n8 = Sequence()
   n9 = Get CurrentButton() -> (CurrentButton:Widget Multi Click Button Object Reference)
   > Set ButtonText!(execute, ButtonText:String=@n6, self:Widget Multi Click Button Object Reference=@n9) -> (then, Output_Get:String)
   > Set TagText!(execute, TagText:String=@n6.Tag_5_#2, self:Widget Multi Click Button Object Reference=@n9) -> (then, Output_Get:String)
   n10 = Get CurrentButton()
   > Set Exclusive!(execute, Exclusive=@n6.Exclusive_39_#3, self:Widget Multi Click Button Object Reference=@n10) -> (then, Output_Get)
   > Set TriState!(execute, TriState:bool=@n6.TriState_32_#4, self:Widget Multi Click Button Object Reference=@n10) -> (then, Output_Get:bool)
   n11 = Get CurrentButton()
   > Set TagMask!(execute, TagMask:Integer64=0, self:Widget Multi Click Button Object Reference=@n11) -> (then, Output_Get:Integer64=0)
   > SetIsEnabled(self=@n11, bInIsEnabled=NOT Boolean(A=@n6.ButtonIsDisabled_27_#9))
   > Set PostTagFunction!(execute, PostTagFunction=@n6.AdditionalTagFunction_35_#10, self:Widget Multi Click Button Object Reference=@n11) -> (then, Output_Get=0)
   n12 = Get CurrentButton()
   > Set ToolTipText!(
    execute
    ToolTipText:Text="To Text (String)"(InString=@n6.ButtonTooltip_61_#11)
    self:Widget Object Reference=@n12
   ) -> (then, Output_Get:Text)
   > Set Subcomponents!(execute, Subcomponents=@n6.SubComponent_53_#13, self:Widget Multi Click Button Object Reference=@n12) -> (then, Output_Get)
   n13 = Get CurrentButton()
   n14 = Add(execute=@n8.then_1, TargetArray:Array of Widget Object References=Get ChildWidgets(), NewItem:"Widget Object Reference (by ref)"=@n13)
   > RegisterButton!(execute, self, Button:Widget Multi Click Button Object Reference=@n13) -> (then) {is=title K2Node_CallFunction}
   > AddChildToGrid!(
    execute
    self:Grid Panel Object Reference=Get GPanelDynamicUIButtons()
    Content:Widget Object Reference="Get (a copy)"(Array:Array of Widget Object References=Get ChildWidgets(), Dimension 1=@n14) -> (Output:Widget Object Reference)
    InRow:int=@n6.GridPanelRow_21_#8
    InColumn:int=@n6.GridPanelColumn_20_#7
   ) -> (then, ReturnValue:Grid Slot Object Reference)
  ]}
 RemoveTagFromActors {nodes [
   n1 = RemoveTagFromActors!() -> (
    then
    Transaction:bool=~
    ActorTagText:String
    UseSelectedActors:bool=~
    ActorArray:Array of Actor Object References
    SubComponents:Array of Strings
    ExclusiveTags:Array of Names
    AllowForTagsHavingNumericData:bool=~
   ) {is=title K2Node_FunctionEntry}
   > n2 = Branch(Condition=@n1.UseSelectedActors)
   > n3 = GetSelectedLevelActors(self=Get EditorActorSubsystem())
   > n4 = Set TargetActors(TargetActors:Array of Actor Object References=@n3) -> (Output_Get:Array of Actor Object References)
   n5 = Branch(execute=[@n4.then, @n25.then], Condition=@n1)
   > n6 = For Each Loop(Array:Array of Actor Object References=Get TargetActors() -> (TargetActors:Array of Actor Object References)) -> (Array Element:Actor Object Reference)
   > TransactObject(Object=@n6)
   > n7 = Branch(Condition=Get AllowForTagsHavingNumericData() -> (AllowForTagsHavingNumericData:bool))
   n8 = Split(SourceString=Get ActorTagText() -> (ActorTagText:String), InStr=:, SearchDir=FromStart)
   > n9 = Branch(Condition=@n8.ReturnValue)
   n10 = Get Tags(self=@n6)
   > n11 = For Each Loop(Array:Array of Names=@n10) -> (Array Element:Name)
   n12 = Remove Item(execute=@n11.Completed, TargetArray:Array of Names=@n10, Item:"Name (by ref)"=REMOVED)
   n13 = Get Tags(self=@n6)
   n14 = TagExclusivityCheck(execute=[@n12.then, @n15.then], ExclusiveTagsToRemove=Get ExclusiveTags(), TargetNameTagArray=@n13, NameArraysOnly=true)
   n15 = Remove Item(execute=[@n7.else, @n9.else], TargetArray=@n13, Item=String To Name(InString=Get ActorTagText()))
   n16 = Branch(execute=@n14.then, Condition=OR Boolean(A=@n14.TagsWereChanged, B=@n15))
   > n17 = Set Tags(Tags=@n14.FilteredNameTagArray, self=@n6)
   n18 = Get ApplyToBPComponents()
   n19 = Length(TargetArray:Array of Strings=Get SubComponents())
   Branch(execute=[@n17.then, @n16.else], Condition=AND Boolean(A="integer > integer"(A=@n19), B=IsChecked(self=@n18)))
   > n20 = For Each Loop(Array:Array of Static Mesh Component Object References=Get Components By Class(self=@n6, ComponentClass=&StaticMeshComponent) -> (ReturnValue:Array of Static Mesh Component Object References)) -> (Array Element:Static Mesh Component Object Reference)
   n21 = String To Name(InString=Split(SourceString=Get ActorTagText(), InStr=:, SearchDir=FromStart))
   > SetEditorProperty(Object=@n20, PropertyName=@n21, PropertyValue:Byte=MakeLiteralByte!(Value:Byte=0) -> (ReturnValue:Byte=0))
   n22 = For Each Loop(
    Exec=@n20.Completed
    Array:Array of Scene Component Object References=Get Components By Class(self=@n6, ComponentClass=&SceneComponent) -> (ReturnValue:Array of Scene Component Object References)
   ) -> (Array Element:Scene Component Object Reference)
   > SetEditorProperty(Object=@n22, PropertyName=@n21, PropertyValue=MakeLiteralByte(Value=0) -> (ReturnValue=0))
   n23 = "To String (Name)"(InName=@n11)
   Branch(execute=@n11.LoopBody, Condition="integer >= integer"(A=FindSubstring(SearchIn=@n23, Substring=@n8)))
   > Set Array Elem(TargetArray=@n10, Index=@n11.Array Index, Item=REMOVED, bSizeToFit=false)
   n24 = For Each Loop(Exec=@n5.else, Array:Array of Actor Object References=Get TargetActors()) -> (Array Element:Actor Object Reference)
   > Remove Item(TargetArray=Get Tags(self=@n24), Item=String To Name(InString=Get ActorTagText()))
   n25 = Set TargetActors(execute=@n2.else, TargetActors=@n1.ActorArray)
  ]}
 SelectActorsUsingTagSubstring {nodes [
   SelectActorsUsingTagSubstring!() -> (then, TagSubstring:String) {is=title K2Node_FunctionEntry}
   > Set LocalPartialTagFound(LocalPartialTagFound=false)
   > n1 = GetAllLevelActors(self=Get EditorActorSubsystem())
   > n2 = For Each Loop(Array=@n1)
   Return Node!(
    execute=@n2.Completed
    ActorSelection:Array of Actor Object References=Get ActorsToSelect() -> (ActorsToSelect:Array of Actor Object References)
   )
   > n3 = For Each Loop with Break!(Exec:Exec, Array:Array of Names=Get Tags(self=@n2), Break:Exec=@n4.then) -> (LoopBody:Exec, Array Element:Name, Array Index:int, Completed:Exec)
   > Branch(Condition=Split(SourceString="To String (Name)"(InName=@n3), InStr=Get TagSubstring() -> (TagSubstring:String), SearchDir=FromStart).ReturnValue)
   > n4 = Add Unique(TargetArray:Array of Actor Object References=Get ActorsToSelect(), NewItem:"Actor Object Reference (by ref)"=@n2)
  ]}
 SelectActorsbyTag {nodes [
   n1 = SelectActorsbyTag!() -> (then, TextTag:String) {is=title K2Node_FunctionEntry}
   > n2 = GetEditorWorld(self=Get UnrealEditorSubsystem())
   > n3 = GetAllActorsWithTag(WorldContextObject=@n2, Tag=String To Name(InString=@n1))
   > SetSelectedLevelActors(self=Get EditorActorSubsystem(), ActorsToSelect=@n3)
   > Return Node!(execute, SelectedActors:Array of Actor Object References=@n3, NumSelectedActors:int=Length(TargetArray:Array of Actor Object References=@n3))
  ]}
 TagExclusivityCheck {nodes [
   TagExclusivityCheck!() -> (then, TargetTagArray:Array of Strings, ExclusiveTagsToRemove:Array of Names, TargetNameTagArray:Array of Names, NameArraysOnly:bool=~) {is=title K2Node_FunctionEntry}
   > Set TagsHaveChanged(TagsHaveChanged:bool=false) -> (Output_Get:bool)
   > n1 = Branch(Condition=Get NameArraysOnly() -> (NameArraysOnly:bool))
   > n2 = For Each Loop(Array:Array of Names=Get ExclusiveTagsToRemove() -> (ExclusiveTagsToRemove:Array of Names)) -> (Array Element:Name)
   > n3 = Remove Item(TargetArray=Get TargetNameTagArray() -> (TargetNameTagArray:Array of Names), Item=@n2)
   > Branch(Condition="Equal (Boolean)"(A:bool=@n3, B:bool=true) -> (ReturnValue:bool=~) {is=title K2Node_PromotableOperator})
   > Set TagsHaveChanged(TagsHaveChanged=true)
   Return Node!(
    execute=@n2.Completed
    FilteredTagArray:Array of Strings
    TagsWereChanged:bool=Get TagsHaveChanged() -> (TagsHaveChanged:bool)
    FilteredNameTagArray:Array of Names=Get TargetNameTagArray()
   )
   n4 = For Each Loop(Exec=@n1.else, Array=Get ExclusiveTagsToRemove())
   > n5 = Remove Item(
    TargetArray:Array of Strings=Get TargetTagArray() -> (TargetTagArray:Array of Strings)
    Item:"String (by ref)"="To String (Name)"(InName=@n4)
   )
   > Branch(Condition="Equal (Boolean)"(A=@n5, B=true) -> (ReturnValue=~))
   > Set TagsHaveChanged(TagsHaveChanged=true)
   Return Node(execute=@n4.Completed, FilteredTagArray=Get TargetTagArray(), TagsWereChanged=Get TagsHaveChanged())
  ]}
 UpdatePredefinedTagButtonHighlights {nodes [
   n1 = UpdatePredefinedTagButtonHighlights!() -> (then, TypedElementSelectionSet:Typed Element Selection Set Object Reference, DebugOutputGroup:String) {is=title K2Node_FunctionEntry}
   > Clear(TargetArray:Array of Names=Get SelectedActorTags())
   > n2 = GetSelectedObjects!(execute, self:Typed Element Selection Set Object Reference=@n1, InRequiredClass:Object Class Reference=&Actor) -> (then, ReturnValue:Array of Object References)
   > n3 = For Each Loop(Array:Array of Object References=@n2) -> (Array Element:Object Reference)
   PrintString(
    execute=@n3.Completed
    InString=Append!(A=Get DebugOutputGroup() -> (DebugOutputGroup:String), B=" Called UpdatePredefinedTagButtonHighlights") -> (ReturnValue)
    bPrintToScreen=false
    bPrintToLog=Get DebugOutput()
   )
   > ButtonTagHighlights() {is=title K2Node_CallFunction}
   n4 = Cast To Actor!(execute=@n3.LoopBody, Object=@n3) -> (then, CastFailed, AsActor:Actor Object Reference) {is=title K2Node_DynamicCast}
   > n5 = For Each Loop(Array:Array of Names=Get Tags(self=@n4)) -> (Array Element:Name)
   > Add Unique(TargetArray:Array of Names=Get SelectedActorTags(), NewItem:"Name (by ref)"=@n5)
   PrintString(
    execute=@n4.CastFailed
    InString="As Actor Cast Failed in [UpdatePredefinedTagButtonHigfhlights]"
    bPrintToScreen=false
    bPrintToLog=Get DebugOutput()
    Duration=4
   )
  ]}
]
variables [
 AddTagName {type=string, default=None}
 AltPressed {type=boolean}
 ButtonClicked_Dispatcher {default=()}
 ChildWidgets {type="array<Widget>", default []}
 CtrlPressed {type=boolean}
 DebugOutput {type=boolean}
 MouseButtonState {type=integer}
 NewActorSelection {type="array<Actor>"}
 SelectedActorTags {type="array<string>", default []}
 ShiftPressed {type=boolean}
 TypedElementSelectionSet {type=TypedElementSelectionSet, default=null}
 UIButtonDataTablesPreviewLevels {
  type="array<Object>"
  default [&/BiomeCore_Tags, &/PCG_Assembly_Tags]
 }
 UIButtonsDataTable {type=Object, default=&BiomeCore_Tags}
 WidgetFolderPath {type=string, default=""}
]
widgets {
 info {rootWidgetClass=&VerticalBox, widgetCount=21}
 widgets [
  VerticalBox_1 {widgetClassPath=&VerticalBox, widget=VerticalBox_1}
  GridPanel_0 {widgetClassPath=&GridPanel, parent=VerticalBox_1, widget=GridPanel_0, properties {rowFill [0,1]}, slot {class=VerticalBoxSlot}}
  GridPanel_86 {
   widgetClassPath=&GridPanel
   parent=GridPanel_0
   widget=GridPanel_86
   properties {columnFill [0.33,0.5,0.33], rowFill [0,1]}
   slot {class=GridSlot}
  }
  ComboBoxTagButtonDefinitions {
   widgetClassPath=/Script/Blutility.EditorUtilityComboBoxString
   bIsVariable=true
   parent=GridPanel_86
   widget=ComboBoxTagButtonDefinitions
   properties {
    bIsFocusable=false
    contentPadding=<ltrb 4 2 4 4>
    font {
     bForceMonospaced=false
     bMaterialIsStencil=false
     fontMaterial=null
     fontObject=&Roboto
     letterSpacing=0
     monospacedWidth=1
     size=12
     skewAmount=0
     typefaceFontName=Regular
     outlineSettings {
      bApplyOutlineToDropShadows=false
      bMiteredCorners=false
      bSeparateFillAlpha=false
      outlineMaterial=null
      outlineSize=0
      outlineColor=<rgba 0 0 0 1>
     }
    }
    foregroundColor {colorUseRule=UseColor_Specified, specifiedColor=<rgba 0.5271 0.5271 0.5271 1>}
    widgetStyle {
     comboButtonStyle {
      downArrowAlign=VAlign_Center
      buttonStyle {
       clickedSlateSound {resourceObject=null}
       disabled {
        bIsDynamicallyLoaded=false
        drawAs=RoundedBox
        imageType=NoImage
        mirroring=NoMirror
        resourceName=None
        resourceObject=null
        tiling=NoTile
        imageSize=<0 0>
        margin=<ltrb 0 0 0 0>
        outlineSettings {
         bUseBrushTransparency=false
         roundingType=FixedRadius
         width=1
         color {colorUseRule=UseColor_ColorTable, specifiedColor=<rgba 0 0 0 0>}
         cornerRadii=<4 4 4 4>
        }
        tintColor {colorUseRule=UseColor_Specified, specifiedColor=<rgba 0.0395 0.0395 0.0395 1>}
        uVRegion {bIsValid=false, max=<0 0>, min=<0 0>}
       }
       disabledForeground {colorUseRule=UseColor_ColorTable, specifiedColor=<rgba 0 0 0 0>}
       hovered {
        bIsDynamicallyLoaded=false
        drawAs=RoundedBox
        imageType=NoImage
        mirroring=NoMirror
        resourceName=None
        resourceObject=null
        tiling=NoTile
        imageSize=<0 0>
        margin=<ltrb 0 0 0 0>
        outlineSettings {
         bUseBrushTransparency=false
         roundingType=FixedRadius
         width=1
         color {colorUseRule=UseColor_ColorTable, specifiedColor=<rgba 0 0 0 0>}
         cornerRadii=<4 4 4 4>
        }
        tintColor {colorUseRule=UseColor_Specified, specifiedColor=<rgba 0.0048 0.0048 0.0048 1>}
        uVRegion {bIsValid=false, max=<0 0>, min=<0 0>}
       }
       hoveredForeground {colorUseRule=UseColor_ColorTable, specifiedColor=<rgba 0 0 0 0>}
       hoveredSlateSound {resourceObject=null}
       normal {
        bIsDynamicallyLoaded=false
        drawAs=RoundedBox
        imageType=NoImage
        mirroring=NoMirror
        resourceName=None
        resourceObject=null
        tiling=NoTile
        imageSize=<0 0>
        margin=<ltrb 0 0 0 0>
        outlineSettings {
         bUseBrushTransparency=false
         roundingType=FixedRadius
         width=1
         color {colorUseRule=UseColor_ColorTable, specifiedColor=<rgba 0 0 0 0>}
         cornerRadii=<4 4 4 4>
        }
        tintColor {colorUseRule=UseColor_Specified, specifiedColor=<rgba 0.0048 0.0048 0.0048 1>}
        uVRegion {bIsValid=false, max=<0 0>, min=<0 0>}
       }
       normalForeground {colorUseRule=UseColor_ColorTable, specifiedColor=<rgba 0 0 0 0>}
       normalPadding=<ltrb 8 1 8 1>
       pressed {
        bIsDynamicallyLoaded=false
        drawAs=RoundedBox
        imageType=NoImage
        mirroring=NoMirror
        resourceName=None
        resourceObject=null
        tiling=NoTile
        imageSize=<0 0>
        margin=<ltrb 0 0 0 0>
        outlineSettings {
         bUseBrushTransparency=false
         roundingType=FixedRadius
         width=1
         color {colorUseRule=UseColor_ColorTable, specifiedColor=<rgba 0 0 0 0>}
         cornerRadii=<4 4 4 4>
        }
        tintColor {colorUseRule=UseColor_Specified, specifiedColor=<rgba 0.0048 0.0048 0.0048 1>}
        uVRegion {bIsValid=false, max=<0 0>, min=<0 0>}
       }
       pressedForeground {colorUseRule=UseColor_ColorTable, specifiedColor=<rgba 0 0 0 0>}
       pressedPadding=<ltrb 8 2 8 0>
       pressedSlateSound {resourceObject=null}
      }
      contentPadding=<ltrb 0 0 0 0>
      downArrowImage {
       bIsDynamicallyLoaded=false
       drawAs=Image
       imageType=Vector
       mirroring=NoMirror
       resourceName=../../../Engine/Content/Slate/Starship/CoreWidgets/ComboBox/wide-chevron-down.svg
       resourceObject=null
       tiling=NoTile
       imageSize=<16 16>
       margin=<ltrb 0 0 0 0>
       outlineSettings {
        bUseBrushTransparency=false
        roundingType=HalfHeightRadius
        width=0
        color {colorUseRule=UseColor_Specified, specifiedColor=<rgba 0 0 0 0>}
        cornerRadii=<0 0 0 0>
       }
       tintColor {colorUseRule=UseColor_Specified, specifiedColor=<rgba 1 1 1 1>}
       uVRegion {bIsValid=false, max=<0 0>, min=<0 0>}
      }
      downArrowPadding=<ltrb 2 2 2 2>
      menuBorderBrush {
       bIsDynamicallyLoaded=false
       drawAs=RoundedBox
       imageType=NoImage
       mirroring=NoMirror
       resourceName=None
       resourceObject=null
       tiling=NoTile
       imageSize=<0 0>
       margin=<ltrb 0 0 0 0>
       outlineSettings {
        bUseBrushTransparency=false
        roundingType=FixedRadius
        width=1
        color {colorUseRule=UseColor_ColorTable, specifiedColor=<rgba 0 0 0 0>}
        cornerRadii=<0 0 0 0>
       }
       tintColor {colorUseRule=UseColor_Specified, specifiedColor=<rgba 0.0395 0.0395 0.0395 1>}
       uVRegion {bIsValid=false, max=<0 0>, min=<0 0>}
      }
      menuBorderPadding=<ltrb 0 0 0 0>
      shadowColorAndOpacity=<rgba 0 0 0 1>
      shadowOffset=<0 0>
     }
     contentPadding=<ltrb 0 0 0 0>
     menuRowPadding=<ltrb 8 8 8 8>
     pressedSlateSound {resourceObject=null}
     selectionChangeSlateSound {resourceObject=null}
    }
   }
   slot {class=GridSlot, properties {horizontalAlignment=HAlign_Left}}
  }
  CheckBoxContentBrowserTagging {
   widgetClassPath=&EditorUtilityCheckBox
   bIsVariable=true
   parent=GridPanel_86
   widget=CheckBoxContentBrowserTagging
   properties {
    horizontalAlignment=HAlign_Left
    isFocusable=false
    toolTipText="When checked the tags will be applied to any selected Blueprint Actors from the content browser window, as opposed to tagging actors in the current level. The Blueprint Actors will be saved after tagging."
   }
   slot {class=GridSlot, properties {column=1, columnSpan=1, horizontalAlignment=HAlign_Left, padding=<ltrb 25 0 0 0>}}
  }
  TextBlock_34 {
   widgetClassPath=&TextBlock
   parent=CheckBoxContentBrowserTagging
   widget=TextBlock_34
   properties {
    text="   Content Browser Asset Tagging"
    font {
     bForceMonospaced=false
     bMaterialIsStencil=false
     fontMaterial=null
     fontObject=&Roboto
     letterSpacing=0
     monospacedWidth=1
     size=12
     skewAmount=0
     typefaceFontName=Regular
     outlineSettings {
      bApplyOutlineToDropShadows=false
      bMiteredCorners=false
      bSeparateFillAlpha=false
      outlineMaterial=null
      outlineSize=0
      outlineColor=<rgba 0 0 0 1>
     }
    }
   }
   slot {class=PanelSlot}
  }
  ApplyToBPComponents {
   widgetClassPath=&EditorUtilityCheckBox
   bIsVariable=true
   parent=GridPanel_86
   widget=ApplyToBPComponents
   properties {
    checkedState=Checked
    horizontalAlignment=HAlign_Right
    isFocusable=false
    renderTransform {angle=0, scale=<1 1>, shear=<0 0>, translation=<25 0>}
   }
   slot {class=GridSlot, properties {column=2, columnSpan=1}}
  }
  TextBlock {
   widgetClassPath=&TextBlock
   parent=ApplyToBPComponents
   widget=TextBlock
   properties {
    text=Apply to BP Components
    font {
     bForceMonospaced=false
     bMaterialIsStencil=false
     fontMaterial=null
     fontObject=&Roboto
     letterSpacing=0
     monospacedWidth=1
     size=12
     skewAmount=0
     typefaceFontName=Regular
     outlineSettings {
      bApplyOutlineToDropShadows=false
      bMiteredCorners=false
      bSeparateFillAlpha=false
      outlineMaterial=null
      outlineSize=0
      outlineColor=<rgba 0 0 0 1>
     }
    }
   }
   slot {class=PanelSlot}
  }
  PanelSpacer_01 {
   widgetClassPath=&Spacer
   parent=GridPanel_86
   widget=PanelSpacer_01
   properties {size=<1 5>}
   slot {class=GridSlot, properties {columnSpan=3, row=1}}
  }
  PanelSpacer {
   widgetClassPath=&Spacer
   parent=GridPanel_86
   widget=PanelSpacer
   properties {size=<1 5>}
   slot {class=GridSlot, properties {columnSpan=3, row=3}}
  }
  EditorUtilityScrollBox {
   widgetClassPath=&EditorUtilityScrollBox
   parent=GridPanel_0
   widget=EditorUtilityScrollBox
   properties {
    bAnimateWheelScrolling=true
    consumeMouseWheel=Always
    orientation=Orient_Horizontal
    scrollbarPadding=<ltrb 3 3 3 3>
    widgetBarStyle {
     thickness=8
     draggedThumbImage {
      bIsDynamicallyLoaded=false
      drawAs=RoundedBox
      imageType=NoImage
      mirroring=NoMirror
      resourceName=None
      resourceObject=null
      tiling=NoTile
      imageSize=<0 0>
      margin=<ltrb 0 0 0 0>
      outlineSettings {
       bUseBrushTransparency=false
       roundingType=FixedRadius
       width=0
       color {colorUseRule=UseColor_Specified, specifiedColor=<rgba 0 0 0 0>}
       cornerRadii=<4 4 4 4>
      }
      tintColor {colorUseRule=UseColor_Specified, specifiedColor=<rgba 0.0028 0.4873 0.4416 1>}
      uVRegion {bIsValid=false, max=<0 0>, min=<0 0>}
     }
     horizontalBackgroundImage {
      bIsDynamicallyLoaded=false
      drawAs=NoDrawType
      imageType=NoImage
      mirroring=NoMirror
      resourceName=None
      resourceObject=null
      tiling=NoTile
      imageSize=<0 0>
      margin=<ltrb 0 0 0 0>
      outlineSettings {
       bUseBrushTransparency=false
       roundingType=HalfHeightRadius
       width=0
       color {colorUseRule=UseColor_Specified, specifiedColor=<rgba 0 0 0 0>}
       cornerRadii=<0 0 0 0>
      }
      tintColor {colorUseRule=UseColor_Specified, specifiedColor=<rgba 1 1 1 1>}
      uVRegion {bIsValid=false, max=<0 0>, min=<0 0>}
     }
     horizontalBottomSlotImage {
      bIsDynamicallyLoaded=false
      drawAs=NoDrawType
      imageType=NoImage
      mirroring=NoMirror
      resourceName=None
      resourceObject=null
      tiling=NoTile
      imageSize=<0 0>
      margin=<ltrb 0 0 0 0>
      outlineSettings {
       bUseBrushTransparency=false
       roundingType=HalfHeightRadius
       width=0
       color {colorUseRule=UseColor_Specified, specifiedColor=<rgba 0 0 0 0>}
       cornerRadii=<0 0 0 0>
      }
      tintColor {colorUseRule=UseColor_Specified, specifiedColor=<rgba 1 1 1 1>}
      uVRegion {bIsValid=false, max=<0 0>, min=<0 0>}
     }
     horizontalTopSlotImage {
      bIsDynamicallyLoaded=false
      drawAs=NoDrawType
      imageType=NoImage
      mirroring=NoMirror
      resourceName=None
      resourceObject=null
      tiling=NoTile
      imageSize=<0 0>
      margin=<ltrb 0 0 0 0>
      outlineSettings {
       bUseBrushTransparency=false
       roundingType=HalfHeightRadius
       width=0
       color {colorUseRule=UseColor_Specified, specifiedColor=<rgba 0 0 0 0>}
       cornerRadii=<0 0 0 0>
      }
      tintColor {colorUseRule=UseColor_Specified, specifiedColor=<rgba 1 1 1 1>}
      uVRegion {bIsValid=false, max=<0 0>, min=<0 0>}
     }
     hoveredThumbImage {
      bIsDynamicallyLoaded=false
      drawAs=RoundedBox
      imageType=NoImage
      mirroring=NoMirror
      resourceName=None
      resourceObject=null
      tiling=NoTile
      imageSize=<0 0>
      margin=<ltrb 0 0 0 0>
      outlineSettings {
       bUseBrushTransparency=false
       roundingType=FixedRadius
       width=0
       color {colorUseRule=UseColor_Specified, specifiedColor=<rgba 0 0 0 0>}
       cornerRadii=<4 4 4 4>
      }
      tintColor {colorUseRule=UseColor_Specified, specifiedColor=<rgba 0.2159 0.2159 0.2159 1>}
      uVRegion {bIsValid=false, max=<0 0>, min=<0 0>}
     }
     normalThumbImage {
      bIsDynamicallyLoaded=false
      drawAs=RoundedBox
      imageType=NoImage
      mirroring=NoMirror
      resourceName=None
      resourceObject=null
      tiling=NoTile
      imageSize=<0 0>
      margin=<ltrb 0 0 0 0>
      outlineSettings {
       bUseBrushTransparency=false
       roundingType=FixedRadius
       width=0
       color {colorUseRule=UseColor_Specified, specifiedColor=<rgba 0 0 0 0>}
       cornerRadii=<4 4 4 4>
      }
      tintColor {colorUseRule=UseColor_Specified, specifiedColor=<rgba 0.0953 0.0953 0.0953 1>}
      uVRegion {bIsValid=false, max=<0 0>, min=<0 0>}
     }
     verticalBackgroundImage {
      bIsDynamicallyLoaded=false
      drawAs=NoDrawType
      imageType=NoImage
      mirroring=NoMirror
      resourceName=None
      resourceObject=null
      tiling=NoTile
      imageSize=<0 0>
      margin=<ltrb 0 0 0 0>
      outlineSettings {
       bUseBrushTransparency=false
       roundingType=HalfHeightRadius
       width=0
       color {colorUseRule=UseColor_Specified, specifiedColor=<rgba 0 0 0 0>}
       cornerRadii=<0 0 0 0>
      }
      tintColor {colorUseRule=UseColor_Specified, specifiedColor=<rgba 1 1 1 1>}
      uVRegion {bIsValid=false, max=<0 0>, min=<0 0>}
     }
     verticalBottomSlotImage {
      bIsDynamicallyLoaded=false
      drawAs=NoDrawType
      imageType=NoImage
      mirroring=NoMirror
      resourceName=None
      resourceObject=null
      tiling=NoTile
      imageSize=<0 0>
      margin=<ltrb 0 0 0 0>
      outlineSettings {
       bUseBrushTransparency=false
       roundingType=HalfHeightRadius
       width=0
       color {colorUseRule=UseColor_Specified, specifiedColor=<rgba 0 0 0 0>}
       cornerRadii=<0 0 0 0>
      }
      tintColor {colorUseRule=UseColor_Specified, specifiedColor=<rgba 1 1 1 1>}
      uVRegion {bIsValid=false, max=<0 0>, min=<0 0>}
     }
     verticalTopSlotImage {
      bIsDynamicallyLoaded=false
      drawAs=NoDrawType
      imageType=NoImage
      mirroring=NoMirror
      resourceName=None
      resourceObject=null
      tiling=NoTile
      imageSize=<0 0>
      margin=<ltrb 0 0 0 0>
      outlineSettings {
       bUseBrushTransparency=false
       roundingType=HalfHeightRadius
       width=0
       color {colorUseRule=UseColor_Specified, specifiedColor=<rgba 0 0 0 0>}
       cornerRadii=<0 0 0 0>
      }
      tintColor {colorUseRule=UseColor_Specified, specifiedColor=<rgba 1 1 1 1>}
      uVRegion {bIsValid=false, max=<0 0>, min=<0 0>}
     }
    }
   }
   slot {class=GridSlot, properties {row=1}}
  }
  MultiClickButtonBorder {
   widgetClassPath=&Border
   bIsVariable=true
   parent=EditorUtilityScrollBox
   widget=MultiClickButtonBorder
   properties {
    horizontalAlignment=HAlign_Left
    background {
     bIsDynamicallyLoaded=false
     drawAs=Image
     imageType=NoImage
     mirroring=NoMirror
     resourceName=None
     resourceObject=null
     tiling=NoTile
     imageSize=<32 32>
     margin=<ltrb 0 0 0 0>
     outlineSettings {
      bUseBrushTransparency=false
      roundingType=HalfHeightRadius
      width=0
      color {colorUseRule=UseColor_Specified, specifiedColor=<rgba 0 0 0 0>}
      cornerRadii=<0 0 0 0>
     }
     tintColor {colorUseRule=UseColor_Specified, specifiedColor=<rgba 0 0 0 1>}
     uVRegion {bIsValid=false, max=<0 0>, min=<0 0>}
    }
   }
   slot {class=ScrollBoxSlot, properties {verticalAlignment=VAlign_Top}}
  }
  GPanelDynamicUIButtons {
   widgetClassPath=&GridPanel
   bIsVariable=true
   parent=MultiClickButtonBorder
   widget=GPanelDynamicUIButtons
   properties {bCanChildrenBeAccessible=false}
   slot {class=BorderSlot, properties {horizontalAlignment=HAlign_Left}}
  }
  InputWidget {
   widgetClassPath=&GridPanel
   parent=GridPanel_0
   widget=InputWidget
   slot {class=GridSlot, properties {row=2, verticalAlignment=VAlign_Top}}
  }
  InputWidgetSwitcher {widgetClassPath=&WidgetSwitcher, bIsVariable=true, parent=InputWidget, widget=InputWidgetSwitcher, slot {class=GridSlot}}
  Widget_InputBlank_94 {
   widgetClassPath=&/Widget_InputBlank
   parent=InputWidgetSwitcher
   widget=Widget_InputBlank_94
   slot {class=WidgetSwitcherSlot}
  }
  Widget_InputFloat {
   widgetClassPath=&/Widget_InputFloat
   parent=InputWidgetSwitcher
   widget=Widget_InputFloat
   slot {class=WidgetSwitcherSlot}
  }
  Widget_InputClassDefaultObject {
   widgetClassPath=&/Widget_InputClassDefaultObject
   parent=InputWidgetSwitcher
   widget=Widget_InputClassDefaultObject
   slot {class=WidgetSwitcherSlot}
  }
  Widget_InputSoftObjectPath {
   widgetClassPath=&/Widget_InputSoftObjectPath
   parent=InputWidgetSwitcher
   widget=Widget_InputSoftObjectPath
   slot {class=WidgetSwitcherSlot}
  }
  Widget_InputString {
   widgetClassPath=&/Widget_InputString
   parent=InputWidgetSwitcher
   widget=Widget_InputString
   slot {class=WidgetSwitcherSlot}
  }
  Widget_InputVector3 {
   widgetClassPath=&/Widget_InputVector3
   parent=InputWidgetSwitcher
   widget=Widget_InputVector3
   slot {class=WidgetSwitcherSlot, properties {horizontalAlignment=HAlign_Center}}
  }
 ]
}
